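# Masked, lit Niagara sprite and mesh material for front-end liquid chrome blobs.
# Panning swirl noise, offset per particle, wiggles bubble mask and normal lookups, with dynamic-parameter erosion and near-camera fade shaping the mask.
# View-space reflection vectors sample a sphere map as fake reflection, tinted by ambient cubemap and particle color, then color-dodged.
# Panning chrome normals, Fresnel-driven curve-atlas hue banding and downward-normal roughness add metallic variation; a static switch adds phased emissive.
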Material /Game/Effects/Fort_Effects/FrontEnd/S22/M_FrontEnd_Chrome_Blobs
graph {
 expressions {
  MaterialExpressionAbs_2 {class=MaterialExpressionAbs, inputs [None {from=MaterialExpressionComponentMask_8, output=""}], properties {}}
  MaterialExpressionAdd_0 {
   class=MaterialExpressionAdd
   inputs [
    A {from=MaterialExpressionMultiply_2}
    B {from=MaterialExpressionMultiply_1}
   ]
  }
  MaterialExpressionAdd_1 {
   class=MaterialExpressionAdd
   inputs [
    A {from=MaterialExpressionTextureSample_0, output=RGB}
    B {from=MaterialExpressionTextureSample_1, output=RGB}
   ]
  }
  MaterialExpressionAdd_13 {
   class=MaterialExpressionAdd
   inputs [
    A {from=MaterialExpressionComponentMask_7}
    B {from=MaterialExpressionNamedRerouteUsage_7}
   ]
  }
  MaterialExpressionAdd_14 {
   class=MaterialExpressionAdd
   inputs [
    A {from=MaterialExpressionTextureSample_10, output=R}
    B {from=MaterialExpressionMaterialFunctionCall_2, output=Fresnel}
   ]
  }
  MaterialExpressionAdd_15 {
   class=MaterialExpressionAdd
   inputs [
    A {from=MaterialExpressionAdd_0}
    B {from=MaterialExpressionNamedRerouteUsage_9}
   ]
  }
  MaterialExpressionAdd_16 {
   class=MaterialExpressionAdd
   inputs [
    A {from=MaterialExpressionDivide_13}
    B {from=MaterialExpressionPower_1}
   ]
  }
  MaterialExpressionAdd_2 {
   class=MaterialExpressionAdd
   inputs [
    A {from=MaterialExpressionTextureCoordinate_4}
    B {from=MaterialExpressionParticleRandom_0}
   ]
  }
  MaterialExpressionAdd_20 {
   class=MaterialExpressionAdd
   inputs [
    A {from=MaterialExpressionMultiply_22}
    B {from=MaterialExpressionMultiply_29}
   ]
  }
  MaterialExpressionAdd_22 {
   class=MaterialExpressionAdd
   inputs [
    A {from=MaterialExpressionTextureSample_12, output=RGB}
    B {from=MaterialExpressionTextureSample_15, output=RGB}
   ]
  }
  MaterialExpressionAdd_4 {
   class=MaterialExpressionAdd
   inputs [
    A {from=MaterialExpressionTextureCoordinate_0}
    B {from=MaterialExpressionMultiply_4}
   ]
  }
  MaterialExpressionAdd_7 {
   class=MaterialExpressionAdd
   inputs [
    A {from=MaterialExpressionDivide_1}
    B {from=MaterialExpressionScalarParameter_59}
   ]
  }
  MaterialExpressionAdd_8 {
   class=MaterialExpressionAdd
   inputs [
    A {from=MaterialExpressionMultiply_9}
    B {from=MaterialExpressionNamedRerouteUsage_13}
   ]
  }
  MaterialExpressionClamp_1 {
   class=MaterialExpressionClamp
   inputs [
    None {from=MaterialExpressionMultiply_18, output=""}
    Min {from="", output=""}
    Max {from="", output=""}
   ]
   properties {maxDefault=15}
  }
  MaterialExpressionClamp_2 {
   class=MaterialExpressionClamp
   inputs [
    None {from=MaterialExpressionAdd_4, output=""}
    Min {from="", output=""}
    Max {from="", output=""}
   ]
   properties {}
  }
  MaterialExpressionClamp_3 {
   class=MaterialExpressionClamp
   inputs [
    None {from=MaterialExpressionMultiply_19, output=""}
    Min {from="", output=""}
    Max {from="", output=""}
   ]
   properties {}
  }
  MaterialExpressionComponentMask_0 {class=MaterialExpressionComponentMask, inputs [None {from=MaterialExpressionVectorParameter_0, output=RGB}], properties {g=true, r=true}}
  MaterialExpressionComponentMask_1 {
   class=MaterialExpressionComponentMask
   inputs [None {from=MaterialExpressionMaterialFunctionCall_7, output=AmbientCubemapTint}]
   properties {b=true, g=true, r=true}
  }
  MaterialExpressionComponentMask_3 {
   class=MaterialExpressionComponentMask
   inputs [None {from=MaterialExpressionTextureSampleParameter2D_1, output=RGB}]
   properties {g=true, r=true}
  }
  MaterialExpressionComponentMask_6 {class=MaterialExpressionComponentMask, inputs [None {from=MaterialExpressionTransform_2}], properties {g=true, r=true}}
  MaterialExpressionComponentMask_7 {class=MaterialExpressionComponentMask, inputs [None {from=MaterialExpressionNamedRerouteUsage_5}], properties {g=true, r=true}}
  MaterialExpressionComponentMask_8 {class=MaterialExpressionComponentMask, inputs [None {from=MaterialExpressionReflectionVectorWS_2}], properties {b=true, r=true}}
  MaterialExpressionConstant2Vector_1 {class=MaterialExpressionConstant2Vector, properties {g=-1, r=-0.66}}
  MaterialExpressionConstant2Vector_2 {class=MaterialExpressionConstant2Vector, properties {g=1400, r=2600}}
  MaterialExpressionConstant2Vector_3 {class=MaterialExpressionConstant2Vector, properties {g=-1, r=1}}
  MaterialExpressionConstant3Vector_3 {class=MaterialExpressionConstant3Vector, properties {constant=<rgba 0 0 -1 0>}}
  MaterialExpressionConstantBiasScale_2 {class=MaterialExpressionConstantBiasScale, inputs [None {from=MaterialExpressionComponentMask_6}]}
  MaterialExpressionConstant_0 {class=MaterialExpressionConstant, properties {r=15}}
  MaterialExpressionConstant_4 {class=MaterialExpressionConstant}
  MaterialExpressionConvert_0 {
   class=MaterialExpressionConvert
   inputs [
    None {from=MaterialExpressionTextureSampleParameter2D_4, output=R}
    None {from=MaterialExpressionMultiply_6}
    None {from=MaterialExpressionTextureSampleParameter2D_4, output=R}
   ]
   properties {
    nodeName=Make Vector3
    convertInputs [
     {
      type=Scalar
      expressionInput {
       expression=&self:MaterialExpressionTextureSampleParameter2D_4
       inputName=None
       mask=1
       maskA=0
       maskB=0
       maskG=0
       maskR=1
       outputIndex=1
      }
     }
     {
      type=Scalar
      expressionInput {
       expression=&self:MaterialExpressionMultiply_6
       inputName=None
       mask=0
       maskA=0
       maskB=0
       maskG=0
       maskR=0
       outputIndex=0
      }
     }
     {
      type=Scalar
      expressionInput {
       expression=&self:MaterialExpressionTextureSampleParameter2D_4
       inputName=None
       mask=1
       maskA=0
       maskB=1
       maskG=0
       maskR=0
       outputIndex=3
      }
     }
    ]
    convertOutputs [{type=Vector3}]
   }
  }
  MaterialExpressionCurveAtlasRowParameter_1 {
   class=MaterialExpressionCurveAtlasRowParameter
   inputs [CurveTime {from=MaterialExpressionFrac_5, output=""}]
   properties {
    atlas=/Game/TechArt/Seasonal/LackLuster/Curves/CA_Lackluster
    curve=/Game/TechArt/Seasonal/LackLuster/Curves/C_Lackluster_Reflection
    parameterName=ReflectionCurve
   }
  }
  MaterialExpressionDesaturation_0 {
   class=MaterialExpressionDesaturation
   inputs [
    None {from=MaterialExpressionAdd_15, output=""}
    Fraction {from=MaterialExpressionScalarParameter_22, output=""}
   ]
   properties {}
  }
  MaterialExpressionDesaturation_1 {
   class=MaterialExpressionDesaturation
   inputs [
    None {from=MaterialExpressionMax_0, output=""}
    Fraction {from=MaterialExpressionScalarParameter_62, output=""}
   ]
   properties {}
  }
  MaterialExpressionDivide_0 {
   class=MaterialExpressionDivide
   inputs [
    A {from=MaterialExpressionTextureCoordinate_3}
    B {from=MaterialExpressionScalarParameter_2}
   ]
  }
  MaterialExpressionDivide_1 {
   class=MaterialExpressionDivide
   inputs [
    A {from=MaterialExpressionVectorParameter_8, output=RGB}
    B {from=MaterialExpressionOneMinus_0}
   ]
  }
  MaterialExpressionDivide_13 {
   class=MaterialExpressionDivide
   inputs [
    A {from=MaterialExpressionTextureCoordinate_5}
    B {from=MaterialExpressionConstant2Vector_2}
   ]
   properties {constB=1400}
  }
  MaterialExpressionDivide_4 {
   class=MaterialExpressionDivide
   inputs [
    A {from=MaterialExpressionTextureCoordinate_1}
    B {from=MaterialExpressionScalarParameter_13}
   ]
  }
  MaterialExpressionDivide_5 {
   class=MaterialExpressionDivide
   inputs [
    A {from=MaterialExpressionTextureCoordinate_2}
    B {from=MaterialExpressionScalarParameter_1}
   ]
  }
  MaterialExpressionDotProduct_4 {
   class=MaterialExpressionDotProduct
   inputs [
    A {from=MaterialExpressionVertexNormalWS_2, output=""}
    B {from=MaterialExpressionConstant3Vector_3, output=""}
   ]
   properties {}
  }
  MaterialExpressionDynamicParameter_0 {class=MaterialExpressionDynamicParameter, inputs [], properties {parameterIndex=3, paramNames [Emissive,Param2,NearFade,Param4]}}
  MaterialExpressionDynamicParameter_1 {
   class=MaterialExpressionDynamicParameter
   inputs []
   properties {defaultValue=<rgba 1 1 1 0>, paramNames [ScaleAlphaByVC, Power, BubbleMaxSize, WPOLerp]}
  }
  MaterialExpressionDynamicParameter_3 {
   class=MaterialExpressionDynamicParameter
   inputs []
   properties {parameterIndex=1, defaultValue=<rgba 1 1 0.2 0>, paramNames [Emissive, Erode, WiggleIntensity, Param4]}
  }
  MaterialExpressionDynamicParameter_5 {
   class=MaterialExpressionDynamicParameter
   inputs []
   properties {parameterIndex=1, defaultValue=<rgba 1 1 0.2 0>, paramNames [Emissive, Erode, WiggleIntensity, Param4]}
  }
  MaterialExpressionDynamicParameter_6 {
   class=MaterialExpressionDynamicParameter
   inputs []
   properties {parameterIndex=1, defaultValue=<rgba 1 1 0.2 0>, paramNames [Emissive, Erode, WiggleIntensity, Param4]}
  }
  MaterialExpressionFrac_2 {class=MaterialExpressionFrac, inputs [None {from=MaterialExpressionMultiply_30, output=""}], properties {}}
  MaterialExpressionFrac_5 {class=MaterialExpressionFrac, inputs [None {from=MaterialExpressionAdd_14, output=""}], properties {}}
  MaterialExpressionLinearInterpolate_0 {
   class=MaterialExpressionLinearInterpolate
   inputs [
    A {from=""}
    B {from=""}
    Alpha {from=MaterialExpressionMaterialFunctionCall_6, output=Fresnel Inverted}
   ]
   properties {constA=0.05}
  }
  MaterialExpressionLinearInterpolate_1 {
   class=MaterialExpressionLinearInterpolate
   inputs [
    A {from=MaterialExpressionScalarParameter_25}
    B {from=MaterialExpressionScalarParameter_0}
    Alpha {from=MaterialExpressionNamedRerouteUsage_0}
   ]
  }
  MaterialExpressionLinearInterpolate_3 {
   class=MaterialExpressionLinearInterpolate
   inputs [
    A {from=MaterialExpressionVectorParameter_11, output=RGB}
    B {from=MaterialExpressionVectorParameter_5, output=RGB}
    Alpha {from=MaterialExpressionMaterialFunctionCall_0, output=Result}
   ]
  }
  MaterialExpressionLinearInterpolate_4 {
   class=MaterialExpressionLinearInterpolate
   inputs [
    A {from=MaterialExpressionMultiply_0}
    B {from=MaterialExpressionTextureSampleParameter2D_2, output=G}
    Alpha {from=""}
   ]
   properties {constAlpha=0.25}
  }
  MaterialExpressionLinearInterpolate_5 {
   class=MaterialExpressionLinearInterpolate
   inputs [
    A {from=MaterialExpressionScalarParameter_45}
    B {from=MaterialExpressionScalarParameter_42}
    Alpha {from=MaterialExpressionMaterialFunctionCall_8, output=Fresnel Inverted}
   ]
  }
  MaterialExpressionMaterialFunctionCall_0 {
   class=MaterialExpressionMaterialFunctionCall
   inputs [Speed {from=""}]
   properties {materialFunction=&DebugTimeSine}
  }
  MaterialExpressionMaterialFunctionCall_1 {
   class=MaterialExpressionMaterialFunctionCall
   inputs [
    Dynamic Param {from=MaterialExpressionDynamicParameter_5, output=Erode}
    Alpha {from=MaterialExpressionMultiply_14}
   ]
   properties {materialFunction=/Game/Effects/Fort_Effects/Materials/_Functions/Erode_Alpha}
  }
  MaterialExpressionMaterialFunctionCall_2 {
   class=MaterialExpressionMaterialFunctionCall
   inputs [
    Brightness {from=""}
    Exponent {from=""}
    Camera Vector {from=""}
    Vertex / Pixel Normal {from=""}
   ]
   properties {materialFunction=&MF_Fresnel_FX}
  }
  MaterialExpressionMaterialFunctionCall_3 {
   class=MaterialExpressionMaterialFunctionCall
   inputs [Distance {from=MaterialExpressionDynamicParameter_0, output=NearFade}]
   properties {materialFunction=&Near_Camera_Fade}
  }
  MaterialExpressionMaterialFunctionCall_4 {
   class=MaterialExpressionMaterialFunctionCall
   inputs [
    Hue Shift Percentage {from=""}
    Texture {from=""}
   ]
   properties {materialFunction=&HueShift}
  }
  MaterialExpressionMaterialFunctionCall_5 {
   class=MaterialExpressionMaterialFunctionCall
   inputs [Emissive {from=MaterialExpressionAdd_8}]
   properties {materialFunction=&MF_FX_EmissiveAdjustments}
  }
  MaterialExpressionMaterialFunctionCall_6 {
   class=MaterialExpressionMaterialFunctionCall
   inputs [
    Brightness {from=""}
    Exponent {from=""}
    Camera Vector {from=""}
    Vertex / Pixel Normal {from=""}
   ]
   properties {materialFunction=&MF_Fresnel_FX}
  }
  MaterialExpressionMaterialFunctionCall_7 {
   class=MaterialExpressionMaterialFunctionCall
   inputs []
   properties {materialFunction=/Engine/Functions/GetPostProcessFunctions/GetAmbientCubemapTint}
  }
  MaterialExpressionMaterialFunctionCall_8 {
   class=MaterialExpressionMaterialFunctionCall
   inputs [
    Brightness {from=""}
    Exponent {from=""}
    Camera Vector {from=""}
    Vertex / Pixel Normal {from=""}
   ]
   properties {materialFunction=&MF_Fresnel_FX}
  }
  MaterialExpressionMax_0 {
   class=MaterialExpressionMax
   inputs [
    A {from=MaterialExpressionTextureSample_16, output=RGB}
    B {from=MaterialExpressionTextureSample_17, output=RGB}
   ]
  }
  MaterialExpressionMultiply_0 {
   class=MaterialExpressionMultiply
   inputs [
    A {from=MaterialExpressionTextureSampleParameter2D_3, output=RGB}
    B {from=MaterialExpressionScalarParameter_7}
   ]
  }
  MaterialExpressionMultiply_1 {
   class=MaterialExpressionMultiply
   inputs [
    A {from=MaterialExpressionVectorParameter_6, output=RGB}
    B {from=MaterialExpressionMaterialFunctionCall_4, output=Result}
   ]
  }
  MaterialExpressionMultiply_11 {
   class=MaterialExpressionMultiply
   inputs [
    A {from=MaterialExpressionClamp_1}
    B {from=MaterialExpressionParticleColor_0, output=RGB}
   ]
  }
  MaterialExpressionMultiply_12 {
   class=MaterialExpressionMultiply
   inputs [
    A {from=MaterialExpressionAdd_7}
    B {from=MaterialExpressionScalarParameter_56}
   ]
  }
  MaterialExpressionMultiply_13 {
   class=MaterialExpressionMultiply
   inputs [
    A {from=MaterialExpressionAdd_16}
    B {from=MaterialExpressionConstant2Vector_1}
   ]
  }
  MaterialExpressionMultiply_14 {
   class=MaterialExpressionMultiply
   inputs [
    A {from=MaterialExpressionTextureSampleParameter2D_2, output=G}
    B {from=MaterialExpressionTextureSampleParameter2D_2, output=G}
   ]
  }
  MaterialExpressionMultiply_15 {
   class=MaterialExpressionMultiply
   inputs [
    A {from=MaterialExpressionReroute_2}
    B {from=MaterialExpressionVectorParameter_3, output=RGB}
   ]
  }
  MaterialExpressionMultiply_17 {
   class=MaterialExpressionMultiply
   inputs [
    A {from=MaterialExpressionVertexInterpolator_0, output=PS}
    B {from=MaterialExpressionMaterialFunctionCall_1, output=Result}
   ]
  }
  MaterialExpressionMultiply_18 {
   class=MaterialExpressionMultiply
   inputs [
    A {from=MaterialExpressionMultiply_0}
    B {from=MaterialExpressionComponentMask_1}
   ]
  }
  MaterialExpressionMultiply_19 {
   class=MaterialExpressionMultiply
   inputs [
    A {from=MaterialExpressionConstantBiasScale_2}
    B {from=MaterialExpressionComponentMask_0}
   ]
  }
  MaterialExpressionMultiply_2 {
   class=MaterialExpressionMultiply
   inputs [
    A {from=MaterialExpressionCurveAtlasRowParameter_1}
    B {from=MaterialExpressionNamedRerouteUsage_8}
   ]
  }
  MaterialExpressionMultiply_20 {
   class=MaterialExpressionMultiply
   inputs [
    A {from=MaterialExpressionAdd_22}
    B {from=MaterialExpressionLinearInterpolate_5}
   ]
  }
  MaterialExpressionMultiply_22 {
   class=MaterialExpressionMultiply
   inputs [
    A {from=MaterialExpressionMultiply_20}
    B {from=MaterialExpressionScalarParameter_54}
   ]
  }
  MaterialExpressionMultiply_23 {
   class=MaterialExpressionMultiply
   inputs [
    A {from=MaterialExpressionConvert_0}
    B {from=MaterialExpressionLinearInterpolate_3}
   ]
  }
  MaterialExpressionMultiply_24 {
   class=MaterialExpressionMultiply
   inputs [
    A {from=MaterialExpressionDesaturation_1}
    B {from=MaterialExpressionScalarParameter_61}
   ]
  }
  MaterialExpressionMultiply_29 {
   class=MaterialExpressionMultiply
   inputs [
    A {from=MaterialExpressionOneMinus_1}
    B {from=MaterialExpressionMultiply_24}
   ]
  }
  MaterialExpressionMultiply_3 {
   class=MaterialExpressionMultiply
   inputs [
    A {from=MaterialExpressionScalarParameter_15}
    B {from=MaterialExpressionTime_1}
   ]
  }
  MaterialExpressionMultiply_30 {
   class=MaterialExpressionMultiply
   inputs [
    A {from=MaterialExpressionMaterialFunctionCall_2, output=Fresnel}
    B {from=MaterialExpressionScalarParameter_53}
   ]
  }
  MaterialExpressionMultiply_31 {
   class=MaterialExpressionMultiply
   inputs [
    A {from=MaterialExpressionAdd_16}
    B {from=MaterialExpressionConstant2Vector_3}
   ]
   properties {constB=-1}
  }
  MaterialExpressionMultiply_4 {
   class=MaterialExpressionMultiply
   inputs [
    A {from=MaterialExpressionComponentMask_3}
    B {from=MaterialExpressionDynamicParameter_6, output=WiggleIntensity}
   ]
  }
  MaterialExpressionMultiply_5 {
   class=MaterialExpressionMultiply
   inputs [
    A {from=MaterialExpressionVectorParameter_2, output=RGB}
    B {from=MaterialExpressionNormalize_3}
   ]
  }
  MaterialExpressionMultiply_6 {
   class=MaterialExpressionMultiply
   inputs [
    A {from=MaterialExpressionTextureSampleParameter2D_4, output=G}
    B {from=""}
   ]
   properties {constB=-1}
  }
  MaterialExpressionMultiply_7 {
   class=MaterialExpressionMultiply
   inputs [
    A {from=MaterialExpressionLinearInterpolate_4}
    B {from=MaterialExpressionPower_0}
   ]
  }
  MaterialExpressionMultiply_9 {
   class=MaterialExpressionMultiply
   inputs [
    A {from=MaterialExpressionNamedRerouteUsage_12}
    B {from=MaterialExpressionDynamicParameter_3, output=Emissive}
   ]
  }
  MaterialExpressionNamedRerouteDeclaration_0 {
   class=MaterialExpressionNamedRerouteDeclaration
   inputs [None {from=MaterialExpressionDivide_0}]
   properties {name=Chrome_UVs_A_NoPanning_1, nodeColor=<rgba 0 1 0.7656 1>}
  }
  MaterialExpressionNamedRerouteDeclaration_11 {
   class=MaterialExpressionNamedRerouteDeclaration
   inputs [None {from=MaterialExpressionSaturate_0}]
   properties {name=Diffuse, nodeColor=<rgba 1 0 0.1172 1>}
  }
  MaterialExpressionNamedRerouteDeclaration_2 {
   class=MaterialExpressionNamedRerouteDeclaration
   inputs [None {from=MaterialExpressionLinearInterpolate_0}]
   properties {name=EdgeNormalFresnelInverted, nodeColor=<rgba 0 1 0.5547 1>}
  }
  MaterialExpressionNamedRerouteDeclaration_21 {
   class=MaterialExpressionNamedRerouteDeclaration
   inputs [None {from=MaterialExpressionDivide_4}]
   properties {name=Chrome_UVs_A_NoPanning, nodeColor=<rgba 0 1 0.7656 1>}
  }
  MaterialExpressionNamedRerouteDeclaration_23 {
   class=MaterialExpressionNamedRerouteDeclaration
   inputs [None {from=MaterialExpressionSphereMask_6}]
   properties {name=VertexNormalMask.-Z, nodeColor=<rgba 0 0.3203 1 1>}
  }
  MaterialExpressionNamedRerouteDeclaration_28 {
   class=MaterialExpressionNamedRerouteDeclaration
   inputs [None {from=MaterialExpressionStaticSwitchParameter_3}]
   properties {name=Emissive, nodeColor=<rgba 1 0 0.8203 1>}
  }
  MaterialExpressionNamedRerouteDeclaration_29 {
   class=MaterialExpressionNamedRerouteDeclaration
   inputs [None {from=MaterialExpressionDesaturation_0}]
   properties {name=BaseColor, nodeColor=<rgba 0 0.6484 1 1>}
  }
  MaterialExpressionNamedRerouteDeclaration_3 {
   class=MaterialExpressionNamedRerouteDeclaration
   inputs [None {from=MaterialExpressionLinearInterpolate_1}]
   properties {name=Roughness, nodeColor=<rgba 1 0.3984 0 1>}
  }
  MaterialExpressionNamedRerouteDeclaration_5 {
   class=MaterialExpressionNamedRerouteDeclaration
   inputs [None {from=MaterialExpressionPanner_0}]
   properties {name=Chrome_UVs_A, nodeColor=<rgba 0.9219 0 1 1>}
  }
  MaterialExpressionNamedRerouteDeclaration_6 {
   class=MaterialExpressionNamedRerouteDeclaration
   inputs [None {from=MaterialExpressionPanner_3}]
   properties {name=Chrome_UVs_B, nodeColor=<rgba 0 0.3906 1 1>}
  }
  MaterialExpressionNamedRerouteDeclaration_7 {
   class=MaterialExpressionNamedRerouteDeclaration
   inputs [None {from=MaterialExpressionTransform_0}]
   properties {name=ChromeNormalsWS, nodeColor=<rgba 1 0 0.5391 1>}
  }
  MaterialExpressionNamedRerouteUsage_0 {class=MaterialExpressionNamedRerouteUsage}
  MaterialExpressionNamedRerouteUsage_12 {class=MaterialExpressionNamedRerouteUsage}
  MaterialExpressionNamedRerouteUsage_13 {class=MaterialExpressionNamedRerouteUsage}
  MaterialExpressionNamedRerouteUsage_2 {class=MaterialExpressionNamedRerouteUsage}
  MaterialExpressionNamedRerouteUsage_3 {class=MaterialExpressionNamedRerouteUsage}
  MaterialExpressionNamedRerouteUsage_4 {class=MaterialExpressionNamedRerouteUsage}
  MaterialExpressionNamedRerouteUsage_47 {class=MaterialExpressionNamedRerouteUsage}
  MaterialExpressionNamedRerouteUsage_5 {class=MaterialExpressionNamedRerouteUsage}
  MaterialExpressionNamedRerouteUsage_6 {class=MaterialExpressionNamedRerouteUsage}
  MaterialExpressionNamedRerouteUsage_7 {class=MaterialExpressionNamedRerouteUsage}
  MaterialExpressionNamedRerouteUsage_8 {class=MaterialExpressionNamedRerouteUsage}
  MaterialExpressionNamedRerouteUsage_9 {class=MaterialExpressionNamedRerouteUsage}
  MaterialExpressionNormalize_2 {class=MaterialExpressionNormalize, inputs [VectorInput {from=MaterialExpressionMultiply_15, output=""}]}
  MaterialExpressionNormalize_3 {class=MaterialExpressionNormalize, inputs [VectorInput {from=MaterialExpressionAdd_1, output=""}]}
  MaterialExpressionOneMinus_0 {class=MaterialExpressionOneMinus, inputs [None {from=MaterialExpressionMultiply_11}]}
  MaterialExpressionOneMinus_1 {class=MaterialExpressionOneMinus, inputs [None {from=MaterialExpressionScalarParameter_54}]}
  MaterialExpressionPanner_0 {
   class=MaterialExpressionPanner
   inputs [
    Coordinate {from=MaterialExpressionNamedRerouteDeclaration_21}
    Time {from=MaterialExpressionMultiply_3}
    Speed {from=""}
   ]
   properties {bFractionalPart=true, speedX=0.0008, speedY=0.0022}
  }
  MaterialExpressionPanner_2 {
   class=MaterialExpressionPanner
   inputs [
    Coordinate {from=MaterialExpressionAdd_2}
    Time {from=""}
    Speed {from=""}
   ]
   properties {speedY=0.4}
  }
  MaterialExpressionPanner_3 {
   class=MaterialExpressionPanner
   inputs [
    Coordinate {from=MaterialExpressionDivide_5}
    Time {from=MaterialExpressionMultiply_3}
    Speed {from=""}
   ]
   properties {bFractionalPart=true, speedX=0.0054, speedY=0.0022}
  }
  MaterialExpressionPanner_6 {
   class=MaterialExpressionPanner
   inputs [
    Coordinate {from=MaterialExpressionAdd_16}
    Time {from=""}
    Speed {from=""}
   ]
   properties {bFractionalPart=true, speedX=0.7056, speedY=-0.0154}
  }
  MaterialExpressionPanner_7 {
   class=MaterialExpressionPanner
   inputs [
    Coordinate {from=MaterialExpressionMultiply_13}
    Time {from=""}
    Speed {from=""}
   ]
   properties {bFractionalPart=true, speedX=-1.2908, speedY=-0.0195}
  }
  MaterialExpressionPanner_8 {
   class=MaterialExpressionPanner
   inputs [
    Coordinate {from=MaterialExpressionAdd_16}
    Time {from=""}
    Speed {from=""}
   ]
   properties {speedX=0.061, speedY=0.01}
  }
  MaterialExpressionPanner_9 {
   class=MaterialExpressionPanner
   inputs [
    Coordinate {from=MaterialExpressionMultiply_31}
    Time {from=""}
    Speed {from=""}
   ]
   properties {bFractionalPart=true, speedX=0.045, speedY=0.013}
  }
  MaterialExpressionParticleColor_0 {class=MaterialExpressionParticleColor, inputs [], properties {}}
  MaterialExpressionParticleRandom_0 {class=MaterialExpressionParticleRandom, inputs [], properties {}}
  MaterialExpressionPower_0 {
   class=MaterialExpressionPower
   inputs [
    Base {from=MaterialExpressionTextureSampleParameter2D_2, output=G}
    Exp {from=MaterialExpressionDynamicParameter_1, output=Power}
   ]
   properties {constExponent=6}
  }
  MaterialExpressionPower_1 {
   class=MaterialExpressionPower
   inputs [
    Base {from=MaterialExpressionAbs_2}
    Exp {from=""}
   ]
   properties {constExponent=3}
  }
  MaterialExpressionReflectionVectorWS_2 {class=MaterialExpressionReflectionVectorWS, inputs [CustomWorldNormal {from="", output=""}], properties {}}
  MaterialExpressionReflectionVectorWS_3 {class=MaterialExpressionReflectionVectorWS, inputs [CustomWorldNormal {from=MaterialExpressionMultiply_23, output=""}], properties {}}
  MaterialExpressionReroute_0 {class=MaterialExpressionReroute, inputs [None {from=MaterialExpressionAdd_8}]}
  MaterialExpressionReroute_1 {class=MaterialExpressionReroute, inputs [None {from=MaterialExpressionReroute_0}]}
  MaterialExpressionReroute_2 {class=MaterialExpressionReroute, inputs [None {from=MaterialExpressionConvert_0}]}
  MaterialExpressionSaturate_0 {class=MaterialExpressionSaturate, inputs [None {from=MaterialExpressionMultiply_12}]}
  MaterialExpressionSaturate_3 {class=MaterialExpressionSaturate, inputs [None {from=MaterialExpressionDotProduct_4}]}
  MaterialExpressionScalarParameter_0 {class=MaterialExpressionScalarParameter, properties {defaultValue=0.22, parameterName=Roughness}}
  MaterialExpressionScalarParameter_1 {class=MaterialExpressionScalarParameter, properties {defaultValue=2352, parameterName=Chrome_Scale_UV_b}}
  MaterialExpressionScalarParameter_10 {class=MaterialExpressionScalarParameter, properties {defaultValue=0.2, parameterName=VertexNormalMaskRadius}}
  MaterialExpressionScalarParameter_13 {class=MaterialExpressionScalarParameter, properties {defaultValue=1680, parameterName=Chrome_Scale_UV_a}}
  MaterialExpressionScalarParameter_15 {class=MaterialExpressionScalarParameter, properties {defaultValue=1, parameterName=ChromePanSpeedScale}}
  MaterialExpressionScalarParameter_19 {class=MaterialExpressionScalarParameter, properties {defaultValue=0.5, parameterName=VertexNormalMaskHardness}}
  MaterialExpressionScalarParameter_2 {class=MaterialExpressionScalarParameter, properties {defaultValue=1680, parameterName=Chrome_Scale_UV_a}}
  MaterialExpressionScalarParameter_22 {class=MaterialExpressionScalarParameter, properties {defaultValue=-0.5, parameterName=Fraction}}
  MaterialExpressionScalarParameter_25 {class=MaterialExpressionScalarParameter, properties {defaultValue=0.28, parameterName=Roughness.-Z}}
  MaterialExpressionScalarParameter_3 {class=MaterialExpressionScalarParameter, properties {defaultValue=2.5, parameterName=EdgeNormal_FresnelExponent}}
  MaterialExpressionScalarParameter_38 {class=MaterialExpressionScalarParameter, properties {defaultValue=1.5, parameterName=FresnelHueShiftExponent}}
  MaterialExpressionScalarParameter_4 {class=MaterialExpressionScalarParameter, properties {defaultValue=0.2, parameterName=WiggleIntensity}}
  MaterialExpressionScalarParameter_42 {class=MaterialExpressionScalarParameter, properties {defaultValue=0.5, parameterName=PhaseEmissiveScaleHigh}}
  MaterialExpressionScalarParameter_45 {class=MaterialExpressionScalarParameter, properties {defaultValue=0.2, parameterName=PhaseEmissiveScaleLow}}
  MaterialExpressionScalarParameter_5 {class=MaterialExpressionScalarParameter, properties {defaultValue=0.1, parameterName=VertexNormalMaskCenter}}
  MaterialExpressionScalarParameter_52 {class=MaterialExpressionScalarParameter, properties {defaultValue=2, parameterName=PhaseFresnelExponent}}
  MaterialExpressionScalarParameter_53 {class=MaterialExpressionScalarParameter, properties {defaultValue=1, parameterName=HueShiftScale}}
  MaterialExpressionScalarParameter_54 {class=MaterialExpressionScalarParameter, properties {parameterName=PhaseActive}}
  MaterialExpressionScalarParameter_56 {class=MaterialExpressionScalarParameter, properties {defaultValue=1.6, parameterName=DiffuseScale}}
  MaterialExpressionScalarParameter_59 {class=MaterialExpressionScalarParameter, properties {defaultValue=-0.5, parameterName=DiffuseBias}}
  MaterialExpressionScalarParameter_61 {class=MaterialExpressionScalarParameter, properties {defaultValue=0.1, parameterName=PhaseBaseEmissiveScale}}
  MaterialExpressionScalarParameter_62 {class=MaterialExpressionScalarParameter, properties {defaultValue=0.5, parameterName=PhaseBaseDesaturation}}
  MaterialExpressionScalarParameter_7 {class=MaterialExpressionScalarParameter, properties {defaultValue=0.5, group=reflection, parameterName=Reflection Multiplier}}
  MaterialExpressionSphereMask_6 {
   class=MaterialExpressionSphereMask
   inputs [
    A {from=MaterialExpressionSaturate_3, output=""}
    B {from=MaterialExpressionScalarParameter_5, output=""}
    Radius {from=MaterialExpressionScalarParameter_10, output=""}
    Hardness {from=MaterialExpressionScalarParameter_19, output=""}
   ]
   properties {}
  }
  MaterialExpressionStaticSwitchParameter_0 {
   class=MaterialExpressionStaticSwitchParameter
   inputs [
    True {from="", output=""}
    False {from="", output=""}
   ]
   properties {parameterName=IsSimpleBubble?}
  }
  MaterialExpressionStaticSwitchParameter_3 {
   class=MaterialExpressionStaticSwitchParameter
   inputs [
    True {from=MaterialExpressionAdd_20, output=""}
    False {from=MaterialExpressionConstant_4, output=""}
   ]
   properties {parameterName=IsPhased}
  }
  MaterialExpressionTextureCoordinate_0 {class=MaterialExpressionTextureCoordinate}
  MaterialExpressionTextureCoordinate_1 {class=MaterialExpressionTextureCoordinate}
  MaterialExpressionTextureCoordinate_2 {class=MaterialExpressionTextureCoordinate}
  MaterialExpressionTextureCoordinate_3 {class=MaterialExpressionTextureCoordinate}
  MaterialExpressionTextureCoordinate_4 {class=MaterialExpressionTextureCoordinate, properties {uTiling=0.4, vTiling=0.4}}
  MaterialExpressionTextureCoordinate_5 {class=MaterialExpressionTextureCoordinate}
  MaterialExpressionTextureSampleParameter2D_0 {
   class=MaterialExpressionTextureSampleParameter2D
   inputs [
    UVs {from=MaterialExpressionClamp_3, output=""}
    Apply View MipBias {from="", output=""}
   ]
   properties {parameterName=ErodeMap, texture=/Game/Effects/Fort_Effects/Textures/Water/T_BoilBubble_D}
  }
  MaterialExpressionTextureSampleParameter2D_1 {
   class=MaterialExpressionTextureSampleParameter2D
   inputs [
    UVs {from=MaterialExpressionPanner_2, output=""}
    Apply View MipBias {from="", output=""}
   ]
   properties {parameterName=DistortTexture, samplerType=SAMPLERTYPE_Normal, texture=/Game/Global/Textures/Default/Noise/T_SwirlNoise_N}
  }
  MaterialExpressionTextureSampleParameter2D_2 {
   class=MaterialExpressionTextureSampleParameter2D
   inputs [
    UVs {from=MaterialExpressionClamp_2, output=""}
    Apply View MipBias {from="", output=""}
   ]
   properties {parameterName=LiquidMaskTexture, texture=&T_BoilBubble_D}
  }
  MaterialExpressionTextureSampleParameter2D_3 {
   class=MaterialExpressionTextureSampleParameter2D
   inputs [
    UVs {from=MaterialExpressionClamp_3, output=""}
    Apply View MipBias {from="", output=""}
   ]
   properties {
    parameterName=SphereReflectionMap
    samplerType=SAMPLERTYPE_LinearGrayscale
    texture=/Game/Effects/Fort_Effects/Textures/T_Emote_Monarch_Sphere
   }
  }
  MaterialExpressionTextureSampleParameter2D_4 {
   class=MaterialExpressionTextureSampleParameter2D
   inputs [
    UVs {from=MaterialExpressionClamp_2, output=""}
    Apply View MipBias {from="", output=""}
   ]
   properties {parameterName=LiquidNormalTexture, samplerType=SAMPLERTYPE_Normal, texture=/Game/Effects/Fort_Effects/Textures/Water/T_BoilBubble_Fix_N}
  }
  MaterialExpressionTextureSample_0 {
   class=MaterialExpressionTextureSample
   inputs [
    UVs {from=MaterialExpressionNamedRerouteUsage_3}
    Tex {from=""}
    Apply View MipBias {from=""}
   ]
   properties {
    samplerSource=SSM_Wrap_WorldGroupSettings
    samplerType=SAMPLERTYPE_Normal
    texture=/Game/TechArt/Seasonal/LackLuster/Textures/T_Lackluster_Swirl_N
   }
  }
  MaterialExpressionTextureSample_1 {
   class=MaterialExpressionTextureSample
   inputs [
    UVs {from=MaterialExpressionNamedRerouteUsage_4}
    Tex {from=""}
    Apply View MipBias {from=""}
   ]
   properties {
    samplerSource=SSM_Wrap_WorldGroupSettings
    samplerType=SAMPLERTYPE_Normal
    texture=&T_Lackluster_Swirl_N
   }
  }
  MaterialExpressionTextureSample_10 {
   class=MaterialExpressionTextureSample
   inputs [
    UVs {from=MaterialExpressionAdd_13}
    Tex {from=""}
    Apply View MipBias {from=""}
   ]
   properties {samplerType=SAMPLERTYPE_LinearColor, texture=/Game/TechArt/Seasonal/LackLuster/Textures/T_Lackluster_Streaks}
  }
  MaterialExpressionTextureSample_12 {
   class=MaterialExpressionTextureSample
   inputs [
    UVs {from=MaterialExpressionPanner_6}
    Tex {from=""}
    Apply View MipBias {from=""}
   ]
   properties {samplerSource=SSM_Wrap_WorldGroupSettings, texture=/Game/TechArt/Seasonal/LackLuster/Textures/T_Lackluster_Phased}
  }
  MaterialExpressionTextureSample_15 {
   class=MaterialExpressionTextureSample
   inputs [
    UVs {from=MaterialExpressionPanner_7}
    Tex {from=""}
    Apply View MipBias {from=""}
   ]
   properties {samplerSource=SSM_Wrap_WorldGroupSettings, texture=&T_Lackluster_Phased}
  }
  MaterialExpressionTextureSample_16 {
   class=MaterialExpressionTextureSample
   inputs [
    UVs {from=MaterialExpressionPanner_8}
    Tex {from=""}
    Apply View MipBias {from=""}
   ]
   properties {samplerSource=SSM_Wrap_WorldGroupSettings, texture=&T_Lackluster_Phased}
  }
  MaterialExpressionTextureSample_17 {
   class=MaterialExpressionTextureSample
   inputs [
    UVs {from=MaterialExpressionPanner_9}
    Tex {from=""}
    Apply View MipBias {from=""}
   ]
   properties {samplerSource=SSM_Wrap_WorldGroupSettings, texture=&T_Lackluster_Phased}
  }
  MaterialExpressionTime_1 {class=MaterialExpressionTime}
  MaterialExpressionTransform_0 {class=MaterialExpressionTransform, inputs [None {from=MaterialExpressionMultiply_5}]}
  MaterialExpressionTransform_1 {class=MaterialExpressionTransform, inputs [None {from=MaterialExpressionNormalize_2}]}
  MaterialExpressionTransform_2 {
   class=MaterialExpressionTransform
   inputs [None {from=MaterialExpressionReflectionVectorWS_3}]
   properties {transformSourceType=TRANSFORMSOURCE_World, transformType=TRANSFORM_View}
  }
  MaterialExpressionVectorParameter_0 {class=MaterialExpressionVectorParameter, properties {parameterName=CubeMapOffset, defaultValue=<rgba 1 1 1 1>}}
  MaterialExpressionVectorParameter_1 {class=MaterialExpressionVectorParameter, properties {parameterName=AlbedoHueShift, defaultValue=<rgba 0.8212 0.9028 1 0>}}
  MaterialExpressionVectorParameter_11 {class=MaterialExpressionVectorParameter, properties {parameterName=ScaleNormal, defaultValue=<rgba 0.5 0.5 2 0>}}
  MaterialExpressionVectorParameter_2 {class=MaterialExpressionVectorParameter, properties {parameterName=EdgeNormalScale, defaultValue=<rgba 0.18 0.18 1 0>}}
  MaterialExpressionVectorParameter_3 {class=MaterialExpressionVectorParameter, properties {parameterName=NormalStrength, defaultValue=<rgba 15 15 1 1>}}
  MaterialExpressionVectorParameter_5 {class=MaterialExpressionVectorParameter, properties {parameterName=ScaleNormalHigh, defaultValue=<rgba 2 2 1 0>}}
  MaterialExpressionVectorParameter_6 {class=MaterialExpressionVectorParameter, properties {parameterName=AlbedoScale, defaultValue=<rgba 0.8385 0.8193 0.6886 0>}}
  MaterialExpressionVectorParameter_8 {class=MaterialExpressionVectorParameter, properties {parameterName=Diffuse_ColorDodgeBase, defaultValue=<rgba 0.4601 0.4251 0.4002 0>}}
  MaterialExpressionVertexInterpolator_0 {class=MaterialExpressionVertexInterpolator, inputs [VS {from=MaterialExpressionMaterialFunctionCall_3, output=Result}], properties {}}
  MaterialExpressionVertexNormalWS_0 {class=MaterialExpressionVertexNormalWS, inputs [], properties {}}
  MaterialExpressionVertexNormalWS_2 {class=MaterialExpressionVertexNormalWS, inputs [], properties {}}
 }
 properties {
  bUsedWithNiagaraMeshParticles=true
  bUsedWithNiagaraSprites=true
  blendMode=BLEND_Masked
  pixelDepthOffsetMode=PDOM_Legacy
  usedShadingModels=Default Lit
 }
}
