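# Translucent unlit Niagara ribbon material for the Season 22 front end's chrome ribbon effect, with animated color and shape.
# Two noise layers distort stretched UVs sampling a base shape, whose grayscale blends two colors, then gets hue-shifted by noise, distorted UVs and an opacity mask.
# That noise-warped mask, inverted, multiplies shape opacity, with particle color and depth fade applied.
# A curve atlas row along U drives SplineThicken width (red) and opacity (green).
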Material /Game/Effects/Fort_Effects/FrontEnd/S22/M_FrontEnd_Chrome_Ribbon
graph {
 properties {bUsedWithNiagaraRibbons=true, blendMode=BLEND_Translucent, pixelDepthOffsetMode=PDOM_Legacy, shadingModel=MSM_Unlit, usedShadingModels=Unlit}
 nodes [
  n1 = VectorParameter() {
   group=AdditionalOpacityA
   parameterName=AdditionalOpacityA_UVControl
   channelNames {a=VPan, b=UPan, g=VScale, r=UScale}
   defaultValue=<rgba 1 1 0 0>
  }
  n2 = VectorParameter() {
   group=AdditionalOpacityNoise
   parameterName=AdditionalOpacityNoise_UVControl
   channelNames {a=VPan, b=UPan, g=VScale, r=UScale}
   defaultValue=<rgba 1 1 0 0>
  }
  n3 = MaterialFunctionCall(
   Texture=TextureObjectParameter() {group=AdditionalOpacityNoise, parameterName=AdditionalOpacityNoise_Texture, sortPriority=0}
   UVControl_RGB=@n2
   UVControl_A=@n2
   "[Override]UVInput"=NamedRerouteUsage()
  ) {materialFunction=&FXCommon_TextureSample}
  n4 = VectorParameter() {
   group=AdditionalOpacityNoise
   parameterName=AdditionalOpacityNoise_NoiseControl
   sortPriority=60
   channelNames {a=VStrength, b=UStrength, g=VOffset, r=UOffset}
   defaultValue=<rgba -0.5 -0.5 0.01 0.01>
  }
  n5 = MaterialFunctionCall(
   Texture=TextureObjectParameter() {group=AdditionalOpacityA, parameterName=AdditionalOpacityA_Texture, sortPriority=0}
   UVControl_RGB=@n1
   UVControl_A=@n1
   "[Override]UVInput"=Add(
    A=NamedRerouteUsage()
    B=MaterialFunctionCall!(NoiseValue=@n3.RG, Control_RGB=@n4, Control_A=@n4) {materialFunction=/Game/Effects/Niagara/MasterMaterials/Functions/Common/FXCommon_NoiseControl}.NoiseValue
   )
  ) {materialFunction=&FXCommon_TextureSample}
  n6 = Multiply(A=ScalarParameter() {parameterName=AdditionalOpacityToHueStr}, B=@n5.R)
  n7 = ScalarParameter() {parameterName=AdditionalOpacityToHueoffset}
  n8 = TextureCoordinate()
  n9 = Multiply(
   A=OneMinus(None=MaterialFunctionCall(Float2=@n8) {materialFunction=&BreakOutFloat2Components}.R)
   B=ScalarParameter() {defaultValue=0.6, group=03.BaseShape, parameterName=BaseShape_Stretch}
  )
  n10 = ScalarParameter() {group=03.BaseShape, parameterName=BaseShape_StretchOffset}
  n11 = Add(A=@n8, B=AppendVector(A=Add(A=@n9, B=@n10), B=Constant()))
  n12 = TextureObjectParameter() {
   group=NoiseA
   parameterName=NoiseA_Texture
   samplerType=SAMPLERTYPE_LinearGrayscale
   sortPriority=0
   texture=/Game/Gadgets/Assets/VinderTech_GliderChute/Glider_Thrive_Female/FX/Texture/T_AntiMaterialCharge_Dissolve_01
  }
  n13 = VectorParameter() {group=NoiseA, parameterName=NoiseA_UVControl, channelNames {a=VPan, b=UPan, g=VScale, r=UScale}, defaultValue=<rgba 1 0.5 1 0>}
  n14 = VectorParameter() {
   group=NoiseA
   parameterName=NoiseA_NoiseControl
   sortPriority=60
   channelNames {a=VStrength, b=UStrength, g=VOffset, r=UOffset}
   defaultValue=<rgba -0.5 -0.5 0 -0.5>
  }
  n15 = MaterialFunctionCall(
   NoiseValue=MaterialFunctionCall(Texture=@n12, UVControl_RGB=@n13, UVControl_A=@n13, "[Override]UVInput"=NamedRerouteUsage()) {materialFunction=&FXCommon_TextureSample}.RG
   Control_RGB=@n14
   Control_A=@n14
  ) {materialFunction=&FXCommon_NoiseControl}
  n16 = Add(A=@n11, B=@n15.NoiseValue)
  n17 = ComponentMask(None=@n16) {g=true}
  n18 = ComponentMask(None=@n16) {r=true}
  n19 = Add(
   A=Multiply(A=StaticSwitchParameter(True=@n17, False=@n18) {parameterName=UseGChannel}, B=ScalarParameter() {parameterName=NoiseAToHueShiftStr})
   B=ScalarParameter() {parameterName=NoiseAToHueShiftOffset}
  )
  n20 = VectorParameter() {group=NoiseB, parameterName=NoiseB_UVControl, channelNames {a=VPan, b=UPan, g=VScale, r=UScale}, defaultValue=<rgba 1 1 0 0>}
  n21 = VectorParameter() {
   group=NoiseB
   parameterName=NoiseB_NoiseControl
   sortPriority=60
   channelNames {a=VStrength, b=UStrength, g=VOffset, r=UOffset}
   defaultValue=<rgba -0.5 -0.5 0.01 0.01>
  }
  n22 = MaterialFunctionCall(
   NoiseValue=MaterialFunctionCall(
    Texture=TextureObjectParameter() {group=NoiseB, parameterName=NoiseB_Texture, sortPriority=0}
    UVControl_RGB=@n20
    UVControl_A=@n20
    "[Override]UVInput"=NamedRerouteUsage()
   ) {materialFunction=&FXCommon_TextureSample}.RG
   Control_RGB=@n21
   Control_A=@n21
  ) {materialFunction=&FXCommon_NoiseControl}
  n23 = ComponentMask(None=@n22.NoiseValue) {g=true}
  n24 = ComponentMask(None=@n22.NoiseValue) {r=true}
  n25 = Add(
   A=Multiply(A=StaticSwitchParameter(True=@n23, False=@n24) {parameterName=UseGChannel}, B=ScalarParameter() {parameterName=NoiseBToHueShiftStr})
   B=ScalarParameter() {parameterName=NoiseBToHueShiftOffset}
  )
  n26 = VectorParameter() {group=BaseShape, parameterName=BaseShape_G2C_ColorA, defaultValue=<rgba 0 1 0.7 0>}
  n27 = VectorParameter() {group=BaseShape, parameterName=BaseShape_G2C_ColorB, defaultValue=<rgba 0 1 0 0>}
  n28 = TextureObjectParameter() {
   group=BaseShape
   parameterName=BaseShape_Texture
   samplerType=SAMPLERTYPE_Grayscale
   sortPriority=0
   texture=/Game/Gadgets/Assets/VinderTech_GliderChute/Glider_Thrive_Female/FX/Texture/T_RibbonBaseShape4
  }
  n29 = VectorParameter() {group=BaseShape, parameterName=BaseShape_UVControl, channelNames {a=VPan, b=UPan, g=VScale, r=UScale}, defaultValue=<rgba 1 1 0 0>}
  n30 = MaterialFunctionCall(Texture=@n28, UVControl_RGB=@n29, UVControl_A=@n29, "[Override]UVInput"=Add(A=Add(A=@n15.NoiseValue, B=@n22.NoiseValue), B=@n11)) {materialFunction=&FXCommon_TextureSample}
  n31 = OneMinus(None=Power(Base=OneMinus(None=@n30.R), Exp=ScalarParameter() {defaultValue=1, group=BaseShape, parameterName=BaseShape_G2C_InvtPower}))
  n32 = VectorParameter() {group=BaseShape, parameterName=BaseShape_G2C_Refine, channelNames {a=Blend, b=B_Mul, g=A_Mul, r=A_Power}, defaultValue=<rgba 1 1 1 0.5>}
  n33 = MaterialFunctionCall(Texture=@n31, RGB=@n32, A=@n32) {materialFunction=&MF_FX_Refine}
  n34 = ScalarParameter() {parameterName=BaseShape_G2C_AddAfterRefine}
  n35 = MaterialFunctionCall(
   Hue Shift Percentage=Add(A=Add(A=@n6, B=@n7), B=Add(A=@n19, B=@n25))
   Texture=Multiply(A=LinearInterpolate(A=@n26.RGB, B=@n27.RGB, Alpha=@n31), B=Add(A=@n33.Result, B=@n34))
  ) {materialFunction=&HueShift}
  n36 = VectorParameter() {parameterName=FinalColorCorrect, defaultValue=<rgba 0 1 0.5 1>}
  MaterialFunctionCall(Emissive=Multiply(A=ParticleColor().RGB, B=Multiply(A=@n35.Result, B=@n36.RGB))) {materialFunction=&MF_FX_EmissiveAdjustments}
  n37 = CurveAtlasRowParameter(CurveTime=MaterialFunctionCall(Float2=TextureCoordinate()) {materialFunction=&BreakOutFloat2Components}.R) {
   atlas=&AthenaCurveAtlas01
   curve=/Game/Characters/Player/Male/Medium/Bodies/M_MED_Accumulate/Materials/AccumulateSpriteAnimCurve
   defaultValue=235
   parameterName=RibbonCurve
  }
  n38 = Multiply(A=@n37.R, B=ScalarParameter() {defaultValue=50, parameterName=Width})
  MaterialFunctionCall(WidthBase=@n38, WidthTip=@n38, Expand U or V UV Channel=StaticBool()) {materialFunction=&SplineThicken}
  n39 = Add(A=@n30.R, B=ScalarParameter() {group=BaseShape, parameterName=BaseShape_GrayAdjust_Add})
  n40 = ScalarParameter() {defaultValue=1, group=BaseShape, parameterName=BaseShape_GrayAdjust_Mul}
  n41 = Power(
   Base=Saturate(None=Multiply(A=@n39, B=@n40))
   Exp=ScalarParameter() {defaultValue=1, group=BaseShape, parameterName=BaseShape_GrayAdjust_Power}
  )
  n42 = Multiply(
   A=Add(A=OneMinus(None=@n5.R), B=ScalarParameter() {group=AdditionalOpacityA, parameterName=AdditionalOpacityA_GrayAdjust_Add})
   B=ScalarParameter() {defaultValue=1, group=AdditionalOpacityA, parameterName=AdditionalOpacityA_GrayAdjust_Mul}
  )
  n43 = Multiply(A=Multiply(A=Saturate(None=@n41), B=Add(A=@n42)), B=ScalarParameter() {defaultValue=1, parameterName=FinalOpacityScale})
  Multiply(A=ParticleColor().A, B=DepthFade(Opacity=Multiply(A=@n37.G, B=@n43), FadeDistance=ScalarParameter() {parameterName=DF}))
  NamedRerouteDeclaration(None=MaterialFunctionCall!(
    UVInput=TextureCoordinate()
    TimeInput
    Control_RGB=VectorParameter() {parameterName=TotalUVControl, defaultValue=<rgba 1 1 0 0>}.RGB
    Control_A
   ) {materialFunction=/Game/Effects/Niagara/MasterMaterials/Functions/Common/FXCommon_QuickUVControl}.UV) {name=RemapedUV, nodeColor=<rgba 0 1 0.3672 1>}
 ]
}
