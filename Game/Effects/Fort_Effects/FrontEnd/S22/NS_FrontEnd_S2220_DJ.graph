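# Niagara system for a front-end DJ light show (S2220), with beams, a flare, a spotlight, hexagon sprite rings and stage beams.
# Shared System values for beat, bar, drop and build-up timing drive curves that animate size, color, rotation and beam aim.
# Hex sprites are rebuilt each update as rotating six-sided rings, with HexSpritesLow anchored between BeamOrigin and HeadPosition.
# Long beams and detailed hexes are distance-culled; SpotlightBits is disabled.
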NiagaraSystem /Game/Effects/Fort_Effects/FrontEnd/S22/NS_FrontEnd_S2220_DJ
dependencies {
 usedDataInterfaces [&NiagaraDataInterfaceCurve]
 usedDynamicInputs [&TransformPosition]
 usedModules [
  &self:SystemSpawnScript.SystemScriptSource.SystemScriptGraph.NiagaraNodeAssignment_0.SetVariables_#1
  &SystemState
  &self:SystemSpawnScript.SystemScriptSource.SystemScriptGraph.NiagaraNodeAssignment_3.SetVariables_#2
  &self:SystemSpawnScript.SystemScriptSource.SystemScriptGraph.NiagaraNodeAssignment_1.SetVariables_#3
  &self:SystemSpawnScript.SystemScriptSource.SystemScriptGraph.NiagaraNodeAssignment_2.SetVariables_#4
  &self:SystemSpawnScript.SystemScriptSource.SystemScriptGraph.NiagaraNodeAssignment_4.SetVariables_#5
  &self:SystemSpawnScript.SystemScriptSource.SystemScriptGraph.NiagaraNodeAssignment_6.SetVariables_#6
  &InterpolateOverTime
  &self:SystemSpawnScript.SystemScriptSource.SystemScriptGraph.NiagaraNodeAssignment_8.SetVariables_#7
  &self:SystemSpawnScript.SystemScriptSource.SystemScriptGraph.NiagaraNodeAssignment_5.SetVariables_#8
  &self:SystemSpawnScript.SystemScriptSource.SystemScriptGraph.NiagaraNodeAssignment_9.SetVariables_#9
  &self:SystemSpawnScript.SystemScriptSource.SystemScriptGraph.NiagaraNodeAssignment_7.SetVariables_#10
 ]
}
emitters {
 BeamsLong {
  emitterUpdateScript [
   &EmitterState(
    Enable Distance Culling=true
    MaxDistance=25000
    Module.bUseMaxDistance=true
    Module.bUseMinDistance=false
    Scalability Mode=::NewEnumerator1
   )
   &SpawnRate(SpawnRate=50, Use Spawn Probability=false)
  ]
  particleSpawnScript [
   &InitializeParticle(
    Color=&MultiplyLinearColor(
     Linear Color=&MakeCustomLinearColorFromBool(Bool=&RandomBool(), True Color=<rgba 0.0349 0.0777 1 1>, False Color=<rgba 1.4792 0.0871 1.5 1>)
     Scale Factor=&MakeVector4(X=1, Y=1, Z=1, W=&RandomRangeFloat(Minimum=0.4, Maximum=0.7))
    )
    Lifetime Max=0.5
    Lifetime Min=0.5
    Lifetime Mode=::NewEnumerator1
    Mesh Scale Mode=::NewEnumerator1
    Mesh Uniform Scale Max=3
    Mesh Uniform Scale Min=2
    UsePositionOffset=false
    Write Material Random=true
   )
   &DynamicMaterialParameters(
    Index 0 Param 1=1
    Index 0 Param 2=&RandomRangeFloat(Minimum=2, Maximum=3.5)
    Index 0 Param 3=0.25
    Index 0 Param 4=1
    Module.Param0WriteEnabled=true
    Module.Param1WriteEnabled=false
    Module.Param2WriteEnabled=true
    Module.Param3WriteEnabled=true
   )
   /SetVariables(Particles.RandomBool:NiagaraBool=&RandomBool()) {isSetParameters=true, module=SetVariables_#11}
  ]
  particleUpdateScript [
   &ParticleState()
   &ShapeLocation(
    Cone Angle=1
    MaskSpawnGroup=false
    Override Seed=false
    Rotation Axis=&Lerp_Vector(
     A=&MakeCustomVectorFromBool(Bool=$Particles.RandomBool, True Vector=<0.5 0.25 -0.5>, False Vector=<0.5 -0.25 -0.5>)
     B=&RotateVectorByQuaternion(
      Vector To Rotate=<1 0.5 0>
      Quaternion=&MakeQuatFromAxisAngle(
       Axis=<1 0 -0.4>
       AngleInDegrees=&Add_Float(
        A=&Multiply_Float(A=$System.Age, B=&RandomRangeFloat(Minimum=5, Maximum=10))
        B=&RandomRangeFloat(Minimum=0, Maximum=360)
       )
       NormalizeAxis=true
      )
     )
     Alpha=$System.BuildUp
    )
    Rotation Mode=::NewEnumerator1
    Shape Primitive=::NewEnumerator5
    UseOwnerScale=true
   )
   /Niagara/Modules/Update/Orientation/V2/OrientMeshToVector(Look At Direction:Vector3f=$Particles.ShapeLocation.ShapeNormal, Override Quat=false, UseOptionalSideAxis=false, UseOptionalUpAxis=false)
   &ScaleMeshSize()
   &ScaleColor(
    Scale Alpha=&FloatFromCurve(
     FloatCurve=DataInterface() {propertyValues {Curve {keys [
         {interpMode=RCIM_Cubic, tangentMode=RCTM_User, time=0, value=0}
         {interpMode=RCIM_Cubic, tangentMode=RCTM_User, time=0.5, value=1}
         {tangentMode=RCTM_User, time=1, value=0}
        ]}}}
     Scale Curve=0.1
    )
    Scale RGB=&VectorFromFloat(Value=&Lerp_Float(
      A=0.5
      B=2
      Alpha=&FloatFromCurve(
       FloatCurve=DataInterface() {propertyValues {Curve {keys [<at 0 1 RCIM_Cubic>, {arriveTangent=-1, interpMode=RCIM_Cubic, leaveTangent=-1, time=0.4, value=0.35}, <at 1 0>]}}}
       CurveIndex=$System.DropTime
       Scale Curve=2
      )
     ))
    ScaleA=true
    ScaleRGB=true
   )
   &DynamicMaterialParameters(
    Index 0 Param 1=0.02
    Index 0 Param 2=&FloatFromCurve(
     FloatCurve=DataInterface() {propertyValues {Curve {keys [
         {arriveTangent=5.1604, interpMode=RCIM_Cubic, leaveTangent=5.1604, tangentMode=RCTM_User, time=0, value=0}
         {interpMode=RCIM_Cubic, tangentMode=RCTM_User, time=0.5, value=1}
         {arriveTangent=-4.7059, interpMode=RCIM_Cubic, leaveTangent=-4.7059, tangentMode=RCTM_User, time=1, value=0}
        ]}}}
     Scale Curve=&RandomRangeFloat(Minimum=0.1, Maximum=0.5)
    )
    Module.Param0WriteEnabled=true
    Module.Param1WriteEnabled=true
    Module.Param2WriteEnabled=false
    Module.Param3WriteEnabled=false
   ) {module=DynamicMaterialParameters001}
  ]
  rendererClasses [&NiagaraMeshRendererProperties]
  renderers [{rendererClass=&NiagaraMeshRendererProperties, rendererIndex=0}]
 }
 BeamsLong001 {
  emitterUpdateScript [
   &EmitterState(
    Enable Distance Culling=true
    MaxDistance=25000
    Module.bUseMaxDistance=true
    Module.bUseMinDistance=false
    Scalability Mode=::NewEnumerator1
   )
   &SpawnRate(SpawnRate=25, Use Spawn Probability=false)
  ]
  particleSpawnScript [
   &InitializeParticle(
    Color=&MultiplyLinearColor(
     Linear Color=&MakeCustomLinearColorFromBool(Bool=&RandomBool(), True Color=<rgba 0.0349 0.0777 1 1>, False Color=<rgba 1.4792 0.0871 1.5 1>)
     Scale Factor=&MakeVector4(X=1, Y=1, Z=1, W=&RandomRangeFloat(Minimum=0.4, Maximum=0.7))
    )
    Lifetime Max=0.5
    Lifetime Min=0.5
    Lifetime Mode=::NewEnumerator1
    Mesh Scale Mode=::NewEnumerator1
    Mesh Uniform Scale Max=3
    Mesh Uniform Scale Min=2
    UsePositionOffset=false
    Write Material Random=true
   )
   &DynamicMaterialParameters(
    Index 0 Param 1=1
    Index 0 Param 2=&RandomRangeFloat(Minimum=2, Maximum=3.5)
    Index 0 Param 3=0.25
    Index 0 Param 4=1
    Module.Param0WriteEnabled=true
    Module.Param1WriteEnabled=false
    Module.Param2WriteEnabled=true
    Module.Param3WriteEnabled=true
   )
   /SetVariables(Particles.RandomBool:NiagaraBool=&RandomBool()) {isSetParameters=true, module=SetVariables_#11}
  ]
  particleUpdateScript [
   &ParticleState()
   &ShapeLocation(
    Cone Angle=1
    MaskSpawnGroup=false
    Override Seed=false
    Rotation Axis=&Lerp_Vector(
     A=&MakeCustomVectorFromBool(Bool=$Particles.RandomBool, True Vector=<0.5 0.25 -0.5>, False Vector=<0.5 -0.25 -0.5>)
     B=&RotateVectorByQuaternion(
      Vector To Rotate=<1 0.5 0>
      Quaternion=&MakeQuatFromAxisAngle(
       Axis=<1 0 -0.4>
       AngleInDegrees=&Add_Float(
        A=&Multiply_Float(A=$System.Age, B=&RandomRangeFloat(Minimum=5, Maximum=10))
        B=&RandomRangeFloat(Minimum=0, Maximum=360)
       )
       NormalizeAxis=true
      )
     )
     Alpha=$System.BuildUp
    )
    Rotation Mode=::NewEnumerator1
    Shape Primitive=::NewEnumerator5
    UseOwnerScale=true
   )
   &OrientMeshToVector(Look At Direction:Vector3f=$Particles.ShapeLocation.ShapeNormal, Override Quat=false, UseOptionalSideAxis=false, UseOptionalUpAxis=false)
   &ScaleMeshSize()
   &ScaleColor(
    Scale Alpha=&FloatFromCurve(FloatCurve=DataInterface() {propertyValues {Curve {keys [
         {interpMode=RCIM_Cubic, tangentMode=RCTM_User, time=0, value=0}
         {interpMode=RCIM_Cubic, tangentMode=RCTM_User, time=0.5, value=1}
         {tangentMode=RCTM_User, time=1, value=0}
        ]}}})
    Scale RGB=&VectorFromFloat(Value=&Lerp_Float(
      A=0.5
      B=2
      Alpha=&FloatFromCurve(
       FloatCurve=DataInterface() {propertyValues {Curve {keys [<at 0 1 RCIM_Cubic>, {arriveTangent=-1, interpMode=RCIM_Cubic, leaveTangent=-1, time=0.4, value=0.35}, <at 1 0>]}}}
       CurveIndex=$System.DropTime
      )
     ))
    ScaleA=true
    ScaleRGB=true
   )
   &DynamicMaterialParameters(
    Index 0 Param 1=0.02
    Index 0 Param 2=&FloatFromCurve(
     FloatCurve=DataInterface() {propertyValues {Curve {keys [
         {arriveTangent=5.1604, interpMode=RCIM_Cubic, leaveTangent=5.1604, tangentMode=RCTM_User, time=0, value=0}
         {interpMode=RCIM_Cubic, tangentMode=RCTM_User, time=0.5, value=1}
         {arriveTangent=-4.7059, interpMode=RCIM_Cubic, leaveTangent=-4.7059, tangentMode=RCTM_User, time=1, value=0}
        ]}}}
     Scale Curve=&RandomRangeFloat(Minimum=0.1, Maximum=0.5)
    )
    Module.Param0WriteEnabled=true
    Module.Param1WriteEnabled=true
    Module.Param2WriteEnabled=false
    Module.Param3WriteEnabled=false
   ) {module=DynamicMaterialParameters001}
  ]
  rendererClasses [&NiagaraMeshRendererProperties]
  renderers [{rendererClass=&NiagaraMeshRendererProperties, rendererIndex=0}]
 }
 Flare {
  emitterUpdateScript [
   &EmitterState(
    Inactive Response=::NewEnumerator1
    Life Cycle Mode=::NewEnumerator1
    Loop Behavior=::NewEnumerator1
    Loop Duration Mode=::NewEnumerator1
    UseLoopDelay=false
   )
   &SpawnBurst_Instantaneous(Use Spawn Probability=false, UseLoopCountLimit=false)
  ]
  particleSpawnScript [
   &InitializeParticle(
    Color=<rgba 0.8958 0.0343 1 0.5>
    Sprite Rotation Mode=::NewEnumerator2
    Sprite Size Mode=::NewEnumerator3
    Uniform Sprite Size=200
    UsePositionOffset=false
    Write Material Random=true
   )
   &DynamicMaterialParameters(
    Index 0 Param 1=0.7
    Index 0 Param 2=2
    Index 1 Param 1=500
    Index 3 Param 1=2
    Index 3 Param 2=15
    Module.Param0WriteEnabled=true
    Module.Param0WriteEnabled1=true
    Module.Param0WriteEnabled3=true
    Module.Param1WriteEnabled=true
    Module.Param1WriteEnabled1=false
    Module.Param1WriteEnabled3=true
    Module.Param2WriteEnabled=false
    Module.Param2WriteEnabled1=false
    Module.Param2WriteEnabled3=false
    Module.Param3WriteEnabled=false
    Module.Param3WriteEnabled1=false
    Module.Param3WriteEnabled3=false
    Write Parameter Index 1=true
    Write Parameter Index 3=true
   )
   &CameraOffset(Camera Offset Amount=-50)
   &SystemLocation(MaskSpawnGroup=false, Offset=<10 0 0>)
  ]
  particleUpdateScript [
   &ParticleState(Kill Particles When Lifetime Has Elapsed=false)
   /SetVariables(Particles.Color2=&MultiplyLinearColor(
     Linear Color=<rgba 0.3228 0.13 1 0.35>
     Scale Factor=&MakeFloat4FromVectorAndFloat(
      "Vector (RGB)"=&VectorFromFloat(Value=&MakeFloatFromVector(
        Vector=&VectorFromFloat(Value=&Lerp_Float(
          A=1
          B=1.5
          Alpha=&FloatFromCurve(
           FloatCurve=DataInterface() {propertyValues {Curve {keys [
               {interpMode=RCIM_Cubic, tangentMode=RCTM_User, time=0, value=0}
               {interpMode=RCIM_Cubic, tangentMode=RCTM_User, time=0.5, value=1}
               {tangentMode=RCTM_User, time=1, value=0}
              ]}}}
           CurveIndex=$System.BeatTime
          )
         ))
        Channel=::NewEnumerator0
       ))
      "Float (Alpha)"=1
     )
    )) {isSetParameters=true}
   /SetVariables(Particles.Size2:Vector2f=&Vector2DFromFloat(Value=&Multiply_Float(
      A=&Lerp_Float(
       A=1000
       B=1200
       Alpha=&FloatFromCurve(
        FloatCurve=DataInterface() {propertyValues {Curve {keys [
            {interpMode=RCIM_Cubic, tangentMode=RCTM_User, time=0, value=0}
            {interpMode=RCIM_Cubic, tangentMode=RCTM_User, time=0.5, value=1}
            {tangentMode=RCTM_User, time=1, value=0}
           ]}}}
        CurveIndex=$System.BeatTime
       )
      )
      B=&FloatFromCurve(
       FloatCurve=DataInterface() {propertyValues {Curve {keys [<at 0 0.4>,<at 3000 1>,<at 8000 1>,<at 25000 3>]}}}
       CurveIndex=$Engine.Owner.LODDistance
      )
     ))) {isSetParameters=true}
  ]
  rendererClasses [&NiagaraSpriteRendererProperties]
  renderers [
   {rendererClass=&NiagaraSpriteRendererProperties}
   {rendererClass=&NiagaraSpriteRendererProperties, rendererIndex=1}
  ]
 }
 HexSprites {
  emitterUpdateScript [
   &EmitterState(
    Enable Distance Culling=true
    Inactive Response=::NewEnumerator1
    Life Cycle Mode=::NewEnumerator1
    Loop Behavior=::NewEnumerator1
    Loop Duration Mode=::NewEnumerator1
    Max Distance Response=::SleepAndClearParticles
    MaxDistance=10000
    Module.bUseMaxDistance=true
    Module.bUseMinDistance=false
    Reset Age On Awaken=true
    Scalability Mode=::NewEnumerator1
    UseLoopDelay=false
   )
   &SpawnBurst_Instantaneous(Spawn Count=36, Use Spawn Probability=false, UseLoopCountLimit=false)
  ]
  particleSpawnScript [
   &FN_SetSpawnGroup(Normalize Spawn Group=true, Number of Spawn Groups=6, Particles Per Group=6)
   /SetVariables(Particles.SpawnGroups0-1:NiagaraFloat=&RemapRange(
     Input Value To Remap=&MakeFloatFromInt(Int=&SpawnGroup())
     Input Min=0
     Input Max=5
     Output Min=0
     Output Max=1
    )) {isSetParameters=true, module=SetVariables_#12}
   &InitializeParticle(
    Color=<rgba 1 1 1 0.1>
    Lifetime=5
    Position Mode=::NewEnumerator1
    Position Offset=<0 0 50>
    Sprite Size=<2 2>
    Sprite Size Mode=::NewEnumerator4
    UsePositionOffset=false
    Write Material Random=true
   )
   &DynamicMaterialParameters(
    Index 0 Param 1=2
    Index 0 Param 2=25
    Index 0 Param 3=2
    Module.Param0WriteEnabled=true
    Module.Param1WriteEnabled=true
    Module.Param2WriteEnabled=true
    Module.Param3WriteEnabled=false
   )
  ]
  particleUpdateScript [
   &ParticleState(Kill Particles When Lifetime Has Elapsed=false)
   /SetVariables(Particles.BeatAnim:NiagaraFloat=&FloatFromCurve(
     FloatCurve=DataInterface() {propertyValues {Curve {keys [
         {interpMode=RCIM_Cubic, tangentMode=RCTM_User, time=0, value=0}
         {interpMode=RCIM_Cubic, tangentMode=RCTM_User, time=0.5, value=1}
         {tangentMode=RCTM_User, time=1, value=0}
        ]}}}
     CurveIndex=&Add_Float(
      A=$System.BeatTime
      B=&Lerp_Float(
       A=0
       B=&ClampFloat(Float=&Multiply_Float(A=$System.BeatSpeed, B=-0.125), Min=-0.2, Max=0)
       Alpha=Linked() {linkedVariable {name=Particles.SpawnGroups0-1, type=NiagaraFloat}}
      )
     )
     Scale Curve=0.1
    )) {isSetParameters=true, module=SetVariables_#13}
   /SetVariables(Particles.PositionOrigin:NiagaraPosition=&LerpPosition(
     A=<0 0 0>
     B=<0 0 0>
     Alpha=&Lerp_Float(
      A=&Lerp_Float(A=0, B=0, Alpha=Linked() {linkedVariable {name=Particles.SpawnGroups0-1, type=NiagaraFloat}})
      B=&Lerp_Float(A=0, B=0, Alpha=Linked() {linkedVariable {name=Particles.SpawnGroups0-1, type=NiagaraFloat}})
      Alpha=$System.BuildUp
     )
    )) {isSetParameters=true, module=SetVariables_#14}
   /SetVariables(Particles.Angle:Vector3f=&NormalizeVector(
     Vector To Normalize=&Subtract_Position(A=$Particles.PositionOrigin, B=<1 0.5 0>)
     Safe Normalize=true
     Safe Fallback Threshold=0
     Safe Fallback Vector=<0 0 0.001>
    )) {isSetParameters=true, module=SetVariables_#15}
   /SetVariables(Particles.Size:NiagaraFloat=&Multiply_Float(
     A=&Add_Float(
      A=&Lerp_Float(
       A=&Lerp_Float(A=0.1, B=0.1, Alpha=Linked() {linkedVariable {name=Particles.SpawnGroups0-1, type=NiagaraFloat}})
       B=&Lerp_Float(A=0.5, B=0.02, Alpha=Linked() {linkedVariable {name=Particles.SpawnGroups0-1, type=NiagaraFloat}})
       Alpha=$System.BuildUp
      )
      B=&Multiply_Float(
       A=$Particles.BeatAnim
       B=&Lerp_Float(A=0.1, B=0.1, Alpha=Linked() {linkedVariable {name=Particles.SpawnGroups0-1, type=NiagaraFloat}})
      )
     )
     B=$System.Active
    )) {isSetParameters=true, module=SetVariables_#16}
   /SetVariables(Particles.Offset:Vector3f=&RotateVectorByQuaternion(
     Vector To Rotate=/Niagara/DynamicInputs/Vectors/CrossProduct(
      Vector A:Vector3f=&TransformVector(
       Vector=&MakeVector(
        X=&Multiply_Float(
         A=$Particles.Size
         B=&Add_Float(
          A=225
          B=&FloatFromCurve(
           FloatCurve=DataInterface() {propertyValues {Curve {keys [
               {arriveTangent=99.8455, interpMode=RCIM_Cubic, leaveTangent=99.8455, tangentMode=RCTM_User, time=0, value=0}
               {arriveTangentWeight=0.1307, interpMode=RCIM_Cubic, leaveTangentWeight=0.1859, tangentMode=RCTM_User, time=0.025, value=1}
               {arriveTangent=-4.7766, interpMode=RCIM_Cubic, leaveTangent=-4.7766, tangentMode=RCTM_User, time=0.0649, value=0.9134}
               <at 0.7 0>
              ]}}}
           CurveIndex=&Add_Float(
            A=$System.DropTime
            B=&Lerp_Float(A=0, B=-0.125, Alpha=Linked() {linkedVariable {name=Particles.SpawnGroups0-1, type=NiagaraFloat}})
           )
           Scale Curve=100
          )
         )
        )
        Y=0
        Z=0
       )
       Destination Space=::Simulation
      )
      Vector B:Vector3f=$Particles.Angle
      Normalize Incoming Vectors:NiagaraBool=false
     )
     Quaternion=&MakeQuatFromAxisAngle(
      Axis=$Particles.Angle
      AngleInDegrees=&Add_Float(
       A=&Multiply_Float(A=$Particles.Normalized Spawn Group, B=360)
       B=&Add_Float(
        A=&FloatFromCurve(
         FloatCurve=DataInterface() {propertyValues {Curve {keys [
             {arriveTangent=-2.7362, interpMode=RCIM_Cubic, time=0, value=1}
             {arriveTangent=-8.413, interpMode=RCIM_Cubic, leaveTangent=-8.413, tangentMode=RCTM_User, time=0.084, value=0.65}
             <at 0.28 0>
            ]}}}
         CurveIndex=$System.BarTime
         Scale Curve=60
        )
        B=&FloatFromCurve(
         FloatCurve=DataInterface() {propertyValues {Curve {keys [
             {arriveTangent=-2.7362, interpMode=RCIM_Cubic, time=0, value=1}
             {arriveTangent=-5.446, interpMode=RCIM_Cubic, leaveTangent=-5.446, tangentMode=RCTM_User, time=0.2, value=0.45}
             <at 0.35 0>
            ]}}}
         CurveIndex=$System.BeatTime
         Scale Curve=&Multiply_Float(A=$System.BuildUp, B=60)
        )
       )
      )
     )
    )) {isSetParameters=true, module=SetVariables_#17}
   /SetVariables(Particles.Position=&AddVectorToPosition(Position=$Particles.PositionOrigin, Vector=$Particles.Offset)) {isSetParameters=true, module=SetVariables_#18}
   /SetVariables(Particles.SpriteSize=&MakeVector2D(X=5, Y=&Add_Float(A=&Multiply_Float(A=$Particles.Size, B=100), B=1))) {isSetParameters=true, module=SetVariables_#19}
   &SpriteFacingAndAlignment(
    Sprite Alignment=&CrossProduct(Vector A:Vector3f=$Particles.Offset, Vector B:Vector3f=$Particles.Angle, Normalize Incoming Vectors:NiagaraBool=true)
    WriteAlignment=true
    WriteFacing=false
   )
   &ScaleColor(
    Scale Alpha=&FloatFromCurve(
     FloatCurve=DataInterface() {propertyValues {Curve {keys [
         {interpMode=RCIM_Cubic, tangentMode=RCTM_User, time=0, value=0}
         {interpMode=RCIM_Cubic, tangentMode=RCTM_User, time=0.5, value=0.1}
         {tangentMode=RCTM_User, time=1, value=0}
        ]}}}
     CurveIndex=Linked() {linkedVariable {name=Particles.SpawnGroups0-1, type=NiagaraFloat}}
     Scale Curve=$System.Active
    )
    Scale RGB=&Multiply_VectorByFloat(
     Vector=&MakeVectorFromLinearColorRGB(Color=&LerpLinearColors(
       StartColor=<rgba 0.9244 0.068 1 1>
       EndColor=<rgba 0 4.7725 8 1>
       LerpFactor=&FloatFromCurve(
        FloatCurve=DataInterface() {propertyValues {Curve {keys [
            {interpMode=RCIM_Cubic, tangentMode=RCTM_User, time=0, value=0}
            {interpMode=RCIM_Cubic, tangentMode=RCTM_User, time=0.5, value=1}
            {tangentMode=RCTM_User, time=1, value=0}
           ]}}}
        CurveIndex=$System.DropTime
       )
      ))
     Float=&Lerp_Float(A=0.5, B=2, Alpha=$Particles.BeatAnim)
    )
    ScaleA=true
    ScaleRGB=true
   )
   /SetVariables(Particles.VisibilityTag=&MakeCustomIntFromBool(Bool=&SetBoolByFloatComparison(A=$System.Active, B=0.05), True INT=0, False INT=1)) {isSetParameters=true, module=SetVariables_#20}
  ]
  rendererClasses [&NiagaraSpriteRendererProperties]
  renderers [{rendererClass=&NiagaraSpriteRendererProperties}]
 }
 HexSpritesLow {
  emitterUpdateScript [
   &EmitterState(
    Inactive Response=::NewEnumerator1
    Life Cycle Mode=::NewEnumerator1
    Loop Behavior=::NewEnumerator1
    Loop Duration Mode=::NewEnumerator1
    UseLoopDelay=false
   )
   &SpawnBurst_Instantaneous(Spawn Count=6, Use Spawn Probability=false, UseLoopCountLimit=false)
  ]
  particleSpawnScript [
   &FN_SetSpawnGroup(Normalize Spawn Group=true, Number of Spawn Groups=1, Particles Per Group=6)
   /SetVariables(Particles.SpawnGroups0-1:NiagaraFloat=&RemapRange(
     Input Value To Remap=&MakeFloatFromInt(Int=&SpawnGroup())
     Input Min=0
     Input Max=5
     Output Min=0
     Output Max=1
    )) {isSetParameters=true, module=SetVariables_#12}
   &InitializeParticle(
    Color=<rgba 1 1 1 0.1>
    Lifetime=5
    Position Mode=::NewEnumerator1
    Position Offset=<0 0 50>
    Sprite Size=<1 1>
    Sprite Size Mode=::NewEnumerator4
    UsePositionOffset=false
    Write Material Random=true
   )
   &DynamicMaterialParameters(
    Index 0 Param 1=2
    Index 0 Param 2=25
    Index 0 Param 3=2
    Module.Param0WriteEnabled=true
    Module.Param1WriteEnabled=true
    Module.Param2WriteEnabled=true
    Module.Param3WriteEnabled=false
   )
  ]
  particleUpdateScript [
   &ParticleState(Kill Particles When Lifetime Has Elapsed=false)
   /SetVariables(Particles.BeatAnim:NiagaraFloat=&FloatFromCurve(
     FloatCurve=DataInterface() {propertyValues {Curve {keys [
         {arriveTangent=99.8455, interpMode=RCIM_Cubic, leaveTangent=99.8455, tangentMode=RCTM_User, time=0, value=0}
         {arriveTangent=-5.2513, interpMode=RCIM_Cubic, leaveTangent=-5.2513, tangentMode=RCTM_User, time=0.025, value=1}
         <at 0.5 0>
        ]}}}
     CurveIndex=&Add_Float(
      A=$System.BeatTime
      B=&Lerp_Float(
       A=0
       B=&ClampFloat(Float=&Multiply_Float(A=$System.BeatSpeed, B=-0.125), Min=-0.2, Max=0)
       Alpha=Linked() {linkedVariable {name=Particles.SpawnGroups0-1, type=NiagaraFloat}}
      )
     )
    )) {isSetParameters=true, module=SetVariables_#13}
   /SetVariables(Particles.PositionOrigin:NiagaraPosition=&LerpPosition(
     A=Linked() {linkedVariable {name=System.BeamOrigin, type=Vector3f}}
     B=$System.HeadPosition
     Alpha=&Lerp_Float(A=0.85, B=0.95, Alpha=$System.BuildUp)
    )) {isSetParameters=true, module=SetVariables_#14}
   /SetVariables(Particles.Angle:Vector3f=&NormalizeVector(
     Vector To Normalize=&Subtract_Position(A=$Particles.PositionOrigin, B=$System.HeadPosition)
     Safe Normalize=true
     Safe Fallback Threshold=0
     Safe Fallback Vector=<0 0 0.001>
    )) {isSetParameters=true, module=SetVariables_#15}
   /SetVariables(Particles.Size:NiagaraFloat=&Multiply_Float(
     A=&Add_Float(A=&Lerp_Float(A=2.25, B=4, Alpha=$System.BuildUp), B=$Particles.BeatAnim)
     B=$System.Active
    )) {isSetParameters=true, module=SetVariables_#16}
   /SetVariables(Particles.Offset:Vector3f=&RotateVectorByQuaternion(
     Vector To Rotate=&CrossProduct(
      Vector A:Vector3f=&MakeVector(
       X=&Multiply_Float(
        A=$Particles.Size
        B=&Add_Float(
         A=204
         B=&FloatFromCurve(
          FloatCurve=DataInterface() {propertyValues {Curve {keys [
              {arriveTangent=99.8455, interpMode=RCIM_Cubic, leaveTangent=99.8455, tangentMode=RCTM_User, time=0, value=0}
              {arriveTangentWeight=0.1307, interpMode=RCIM_Cubic, leaveTangentWeight=0.1859, tangentMode=RCTM_User, time=0.025, value=1}
              {arriveTangent=-4.7766, interpMode=RCIM_Cubic, leaveTangent=-4.7766, tangentMode=RCTM_User, time=0.0649, value=0.9134}
              <at 0.7 0>
             ]}}}
          CurveIndex=$System.DropTime
          Scale Curve=500
         )
        )
       )
       Y=0
       Z=0
      )
      Vector B:Vector3f=$Particles.Angle
      Normalize Incoming Vectors:NiagaraBool=false
     )
     Quaternion=&MakeQuatFromAxisAngle(
      Axis=$Particles.Angle
      AngleInDegrees=&Add_Float(
       A=&Multiply_Float(A=$Particles.Normalized Spawn Group, B=360)
       B=&Add_Float(
        A=&FloatFromCurve(
         FloatCurve=DataInterface() {propertyValues {Curve {keys [
             {arriveTangent=-2.7362, interpMode=RCIM_Cubic, time=0, value=1}
             {arriveTangent=-8.413, interpMode=RCIM_Cubic, leaveTangent=-8.413, tangentMode=RCTM_User, time=0.084, value=0.65}
             <at 0.28 0>
            ]}}}
         CurveIndex=$System.BarTime
         Scale Curve=60
        )
        B=&FloatFromCurve(
         FloatCurve=DataInterface() {propertyValues {Curve {keys [
             {arriveTangent=-2.7362, interpMode=RCIM_Cubic, time=0, value=1}
             {arriveTangent=-5.446, interpMode=RCIM_Cubic, leaveTangent=-5.446, tangentMode=RCTM_User, time=0.2, value=0.45}
             <at 0.35 0>
            ]}}}
         CurveIndex=$System.BeatTime
         Scale Curve=&Multiply_Float(A=$System.BuildUp, B=60)
        )
       )
      )
     )
    )) {isSetParameters=true, module=SetVariables_#17}
   /SetVariables(Particles.Position=&AddVectorToPosition(Position=$Particles.PositionOrigin, Vector=$Particles.Offset)) {isSetParameters=true, module=SetVariables_#18}
   /SetVariables(Particles.SpriteSize=&MakeVector2D(X=30, Y=&Add_Float(A=&Multiply_Float(A=$Particles.Size, B=115.4701), B=14))) {isSetParameters=true, module=SetVariables_#19}
   &SpriteFacingAndAlignment(
    Sprite Alignment=&CrossProduct(Vector A:Vector3f=$Particles.Offset, Vector B:Vector3f=$Particles.Angle, Normalize Incoming Vectors:NiagaraBool=true)
    WriteAlignment=true
    WriteFacing=false
   )
   &ScaleColor(
    Scale Alpha=$System.Active
    Scale RGB=&Multiply_VectorByFloat(
     Vector=&MakeVectorFromLinearColorRGB(Color=&LerpLinearColors(
       StartColor=<rgba 0.9244 0.068 1 1>
       EndColor=<rgba 0 4.7725 8 1>
       LerpFactor=&FloatFromCurve(
        FloatCurve=DataInterface() {propertyValues {Curve {keys [
            <at 0 1 RCIM_Cubic>
            {arriveTangent=-2.4971, interpMode=RCIM_Cubic, leaveTangent=-2.4971, tangentMode=RCTM_User, time=0.3, value=0.45}
            <at 1 0>
           ]}}}
        CurveIndex=$System.DropTime
       )
      ))
     Float=&Lerp_Float(A=1, B=10, Alpha=$Particles.BeatAnim)
    )
    ScaleA=true
    ScaleRGB=true
   )
   /SetVariables(Particles.VisibilityTag=&MakeCustomIntFromBool(Bool=&SetBoolByFloatComparison(A=$System.Active, B=0.05), True INT=0, False INT=1)) {isSetParameters=true, module=SetVariables_#20}
  ]
  rendererClasses [&NiagaraSpriteRendererProperties]
  renderers [{rendererClass=&NiagaraSpriteRendererProperties}]
 }
 Spotlight {
  emitterUpdateScript [
   &EmitterState(
    Inactive Response=::NewEnumerator1
    Life Cycle Mode=::NewEnumerator1
    Loop Behavior=::NewEnumerator1
    Loop Duration Mode=::NewEnumerator1
    UseLoopDelay=false
   )
   &SpawnBurst_Instantaneous(Use Spawn Probability=false, UseLoopCountLimit=false)
  ]
  particleSpawnScript [
   &InitializeParticle(
    Color=<rgba 1 0.1073 0.9396 0.5>
    Lifetime Max=0.25
    Lifetime Min=0.2
    Lifetime Mode=::NewEnumerator1
    Mesh Scale Mode=::NewEnumerator3
    Position:Vector3f=$System.BeamOrigin
    Position Mode=::NewEnumerator1
    Sprite Size Mode=::NewEnumerator3
    Uniform Sprite Size=250
    UsePositionOffset=false
    Write Material Random=true
   )
   &CameraOffset(Camera Offset Amount=250)
   /SetVariables(Particles.LightPosition:NiagaraPosition=&AddVectorToPosition(Position=$Particles.Position, Vector=<0 0 0>)) {isSetParameters=true}
   /SetVariables(Particles.BackwardBeamParams:Vector4f=<1 2 0.1 1>) {isSetParameters=true}
  ]
  particleUpdateScript [
   &ParticleState(Kill Particles When Lifetime Has Elapsed=false)
   /SetVariables(Particles.Flicker:NiagaraFloat=&Waveform("Waveform [1]"=::NewEnumerator10, "[1] Frequency"=10, "[1] Amplitude Min/Max"=<0 1>)) {isSetParameters=true}
   &DynamicMaterialParameters(
    Index 0 Param 1=0.1
    Index 0 Param 2=0.5
    Index 0 Param 3=0.25
    Index 0 Param 4=1
    Index 1 Param 1=400
    Index 2 Param 1=2
    Index 2 Param 2=1
    Index 3 Param 1=0.6
    Index 3 Param 2=0.5
    Module.Param0WriteEnabled=true
    Module.Param0WriteEnabled1=true
    Module.Param0WriteEnabled2=true
    Module.Param0WriteEnabled3=true
    Module.Param1WriteEnabled=true
    Module.Param1WriteEnabled1=false
    Module.Param1WriteEnabled2=true
    Module.Param1WriteEnabled3=true
    Module.Param2WriteEnabled=true
    Module.Param2WriteEnabled1=false
    Module.Param2WriteEnabled2=false
    Module.Param2WriteEnabled3=false
    Module.Param3WriteEnabled=true
    Module.Param3WriteEnabled1=false
    Module.Param3WriteEnabled2=false
    Module.Param3WriteEnabled3=false
    Write Parameter Index 1=true
    Write Parameter Index 2=true
    Write Parameter Index 3=true
   )
   &OrientMeshToVector(Look At Direction:Vector3f=<0.1 1 1>, Override Quat=false, UseOptionalSideAxis=false, UseOptionalUpAxis=false)
   &ScaleMeshSize(Scale Factor=&MakeVector(X=-2, Y=1, Z=1))
   &ScaleColor(Scale Alpha=&Lerp_Float(A=0.7, B=1, Alpha=$Particles.Flicker), ScaleA=true, ScaleRGB=false)
   /SetVariables(Particles.LightColor=&MultiplyLinearColor(Linear Color=$Particles.Color, Scale Factor=&MakeVector4(X=1, Y=1, Z=2, W=10))) {isSetParameters=true}
   &SystemLocation(MaskSpawnGroup=false, Offset=<2 -50 100>)
  ]
  rendererClasses [&NiagaraMeshRendererProperties, &NiagaraSpriteRendererProperties]
  renderers [
   {rendererClass=&NiagaraMeshRendererProperties, rendererIndex=0}
   {rendererClass=&NiagaraSpriteRendererProperties, rendererIndex=1}
   {rendererClass=&NiagaraMeshRendererProperties, rendererIndex=2}
  ]
 }
 SpotlightBits {
  bEnabled=false
  simTarget=GPUComputeSim
  emitterUpdateScript [
   &EmitterState(Enable Distance Culling=true, MaxDistance=9000, Module.bUseMaxDistance=true, Module.bUseMinDistance=false, Scalability Mode=::NewEnumerator1)
   &SpawnRate(SpawnRate=20, Use Spawn Probability=false)
  ]
  particleSpawnScript [
   /SetVariables(Particles.Random=&FloatFromCurve(
     FloatCurve=DataInterface() {propertyValues {Curve {keys [
         {interpMode=RCIM_Cubic, tangentMode=RCTM_User, time=0, value=0}
         {arriveTangent=2.173, interpMode=RCIM_Cubic, leaveTangent=2.173, tangentMode=RCTM_User, time=0.6644, value=0.6844}
         {interpMode=RCIM_Cubic, leaveTangent=1.4603, time=1, value=1}
        ]}}}
     CurveIndex=&RandomRangeFloat(Minimum=0, Maximum=1)
    )) {isSetParameters=true}
   &InitializeParticle(
    Color=&LerpLinearColors(StartColor=<rgba 0.316 0.0156 1 0.4>, EndColor=<rgba 0.9048 0.0104 1 1>, LerpFactor=$Particles.Random)
    Lifetime Min=1.4
    Lifetime Mode=::NewEnumerator1
    Sprite Size=&MakeVector2D(
     X=&RandomRangeFloat(Minimum=25, Maximum=35)
     Y=&Lerp_Float(A=22, B=60, Alpha=$Particles.Random)
    )
    Sprite Size Mode=::NewEnumerator4
    UsePositionOffset=false
    Write Material Random=true
   )
   &DynamicMaterialParameters(
    Index 0 Param 1=2
    Index 0 Param 2=100
    Index 0 Param 3=1
    Module.Param0WriteEnabled=true
    Module.Param1WriteEnabled=true
    Module.Param2WriteEnabled=true
    Module.Param3WriteEnabled=false
   )
  ]
  particleUpdateScript [
   &ParticleState()
   /SetVariables(Particles.Direction:Vector3f=&Subtract_Vector(A=$System.BeamOrigin, B=$System.HeadPosition)) {isSetParameters=true}
   /SetVariables(Particles.Position=&Lerp_Vector(
     A=$System.BeamOrigin
     B=$System.HeadPosition
     Alpha=&Add_Float(
      A=&RandomRangeFloat(Minimum=0.1, Maximum=0.9)
      B=&Multiply_Float(A=$Particles.NormalizedAge, B=&Lerp_Float(A=0.07, B=0.11, Alpha=$Particles.Random))
     )
    )) {isSetParameters=true}
   &ScaleColor(
    Scale Alpha=&FloatFromCurve(FloatCurve=DataInterface() {propertyValues {Curve {keys [
         {interpMode=RCIM_Cubic, tangentMode=RCTM_User, time=0, value=0}
         {interpMode=RCIM_Cubic, tangentMode=RCTM_User, time=0.5, value=1}
         {tangentMode=RCTM_User, time=1, value=0}
        ]}}})
    ScaleA=true
    ScaleRGB=false
   )
  ]
  rendererClasses [&NiagaraSpriteRendererProperties]
  renderers [{rendererClass=&NiagaraSpriteRendererProperties}]
 }
 StageBeams {
  emitterUpdateScript [
   &EmitterState(Inactive Response=::NewEnumerator1, Life Cycle Mode=::NewEnumerator1, Loop Duration=1, UseLoopDelay=false)
   &SpawnBurst_Instantaneous(Spawn Count=6, Use Spawn Probability=false, UseLoopCountLimit=false)
   &SpawnBurst_Instantaneous(Spawn Count=6, Spawn Time=1, Use Spawn Probability=false, UseLoopCountLimit=false) {module=SpawnBurst_Instantaneous001}
  ]
  particleSpawnScript [
   &InitializeParticle(
    Color=<rgba 1 1 1 0.5>
    Mesh Scale Mode=::NewEnumerator3
    Position Offset=&RotateVectorByQuaternion(
     Vector To Rotate=<20 0 -50>
     Quaternion=&MakeQuatFromAxisAngle(
      Axis=<0 0 1>
      AngleInDegrees=&Add_Float(A=&ReturnNormalizedExecIndex(Normalized Index Scale=360, Normalized Index Includes One=false), B=30)
     )
    )
    Position Offset Coordinate Space=::World
    Sprite Size Mode=::NewEnumerator3
    Uniform Sprite Size=25
    UsePositionOffset=true
    Write Material Random=true
   )
   &DynamicMaterialParameters(
    Index 0 Param 1=1
    Index 0 Param 2=0.5
    Index 0 Param 3=0.25
    Index 0 Param 4=1
    Index 1 Param 1=20
    Index 3 Param 1=1
    Index 3 Param 2=1
    Module.Param0WriteEnabled=true
    Module.Param0WriteEnabled1=true
    Module.Param0WriteEnabled3=true
    Module.Param1WriteEnabled=true
    Module.Param1WriteEnabled1=false
    Module.Param1WriteEnabled3=true
    Module.Param2WriteEnabled=true
    Module.Param2WriteEnabled1=false
    Module.Param2WriteEnabled3=false
    Module.Param3WriteEnabled=true
    Module.Param3WriteEnabled1=false
    Module.Param3WriteEnabled3=false
    Write Parameter Index 1=true
    Write Parameter Index 3=true
   )
   /SetVariables(Particles.Direction:Vector3f=&NormalizeVector(
     Vector To Normalize=&Multiply_Vector(A=Linked() {linkedVariable {name=Particles.Position, type=NiagaraPosition}}, B=<1 1 1>)
     Safe Normalize=false
     Safe Fallback Threshold=0
     Safe Fallback Vector=<0 0 0.001>
    )) {isSetParameters=true}
  ]
  particleUpdateScript [
   &ParticleState()
   /SetVariables(Particles.BeatAnim:NiagaraFloat=&FloatFromCurve(
     FloatCurve=DataInterface() {propertyValues {Curve {keys [
         {interpMode=RCIM_Cubic, tangentMode=RCTM_User, time=0, value=0}
         {interpMode=RCIM_Cubic, tangentMode=RCTM_User, time=0.1, value=0.95}
         {tangentMode=RCTM_User, time=1, value=0}
        ]}}}
     CurveIndex=$System.BeatTime
     Scale Curve=0
    )) {isSetParameters=true}
   /SetVariables(Particles.DropAnim:NiagaraFloat=&FloatFromCurve(
     FloatCurve=DataInterface() {propertyValues {Curve {keys [
         {interpMode=RCIM_Cubic, tangentMode=RCTM_User, time=0, value=0}
         {interpMode=RCIM_Cubic, tangentMode=RCTM_User, time=0.2, value=1.025}
         {tangentMode=RCTM_User, time=1, value=0}
        ]}}}
     CurveIndex=$System.DropTime
     Scale Curve=0
    )) {isSetParameters=true}
   /SetVariables(Particles.OtherAnim:NiagaraFloat=&FloatFromCurve(
     FloatCurve=DataInterface() {propertyValues {Curve {keys [
         {interpMode=RCIM_Cubic, tangentMode=RCTM_User, time=0, value=0}
         {interpMode=RCIM_Cubic, tangentMode=RCTM_User, time=0.2, value=1.1}
         {tangentMode=RCTM_User, time=1, value=0}
        ]}}}
     CurveIndex=$System.OtherTime
     Scale Curve=0
    )) {isSetParameters=true}
   /Niagara/Modules/Update/Utility/IncrementOverTime(Rate of Change:NiagaraFloat=&Add_Float(A=0, B=&Multiply_Float(A=&Lerp_Float(A=0, B=1, Alpha=$System.BuildUp), B=0)))
   &OrientMeshToVector(
    Look At Direction:Vector3f=&Lerp_Vector(
     A=&RotateVectorByQuaternion(
      Vector To Rotate=&Lerp_Vector(
       A=$Particles.Direction
       B=<0 0 1>
       Alpha=&Lerp_Float(
        A=&Lerp_Float(
         A=&Lerp_Float(
          A=0.65
          B=0.75
          Alpha=&FloatFromCurve(
           FloatCurve=DataInterface() {propertyValues {Curve {keys [
               {interpMode=RCIM_Cubic, tangentMode=RCTM_User, time=0, value=0}
               {interpMode=RCIM_Cubic, tangentMode=RCTM_User, time=0.5, value=1}
               {tangentMode=RCTM_User, time=1, value=0}
              ]}}}
           CurveIndex=$System.BeatTime
          )
         )
         B=0.1
         Alpha=$Particles.OtherAnim
        )
        B=&Waveform("Waveform [1]"=::NewEnumerator0, "[1] Frequency"=1, "[1] Amplitude Min/Max"=<0.6 0.75>)
        Alpha=$System.BuildUp
       )
      )
      Quaternion=&MakeQuatFromAxisAngle(
       Axis=<0 0 1>
       AngleInDegrees=&Lerp_Float(
        A=&Waveform(
         Global Phase=&Add_Float(
          A=$Particles.IncrementOverTime.Counter
          B=&ReturnNormalizedExecIndex(Normalized Index Scale=0.01, Normalized Index Includes One=false)
         )
         Global Amplitude Scale=100
         "Waveform [1]"=::NewEnumerator0
         "[1] Frequency"=1
         "[1] Amplitude Min/Max"=<0 1>
        )
        B=&Multiply_Float(A=&ModuloFloat(A=$Particles.IncrementOverTime.Counter, B=1), B=360)
        Alpha=$System.BuildUp
       )
      )
     )
     B=&RotateVectorByQuaternion(
      Vector To Rotate=&Lerp_Vector(
       A=$Particles.Direction
       B=<0 0 1>
       Alpha=&FloatFromCurve(
        FloatCurve=DataInterface() {propertyValues {Curve {keys [{arriveTangent=-4.3306, interpMode=RCIM_Cubic, leaveTangent=-4.3306, tangentMode=RCTM_User, time=0, value=1}, <at 0.5 0>]}}}
        CurveIndex=$System.DropTime
       )
      )
      Quaternion=&MakeQuatFromAxisAngle(
       Axis=<0 0 1>
       AngleInDegrees=&FloatFromCurve(
        FloatCurve=DataInterface() {propertyValues {Curve {keys [
            {arriveTangent=-5.0148, interpMode=RCIM_Cubic, leaveTangent=-5.0148, tangentMode=RCTM_User, time=0, value=0.5}
            <at 0.2 0>
            {arriveTangent=1.25, interpMode=RCIM_Cubic, leaveTangent=1.25, time=0.4, value=0}
            {interpMode=RCIM_Cubic, leaveTangent=3.6951, time=1, value=1}
           ]}}}
        CurveIndex=$System.DropTime
        Scale Curve=360
       )
      )
     )
     Alpha=&FloatFromCurve(
      FloatCurve=DataInterface() {propertyValues {Curve {keys [
          {interpMode=RCIM_Cubic, tangentMode=RCTM_User, time=0, value=0}
          {interpMode=RCIM_Cubic, tangentMode=RCTM_User, time=0.5, value=1}
          {tangentMode=RCTM_User, time=1, value=0}
         ]}}}
      Scale Curve=0
     )
    )
    Override Quat=false
    UseOptionalSideAxis=false
    UseOptionalUpAxis=false
   )
   &ScaleMeshSize(Scale Factor=&VectorFromFloat(Value=&FloatFromCurve(
      FloatCurve=DataInterface() {propertyValues {Curve {keys [
          {arriveTangent=4.5974, interpMode=RCIM_Cubic, leaveTangent=4.5974, tangentMode=RCTM_User, time=0, value=1}
          <at 0.3 2 RCIM_Cubic>
          {interpMode=RCIM_Cubic, leaveTangent=-3.5436, time=1, value=1}
         ]}}}
      CurveIndex=$System.DropTime
      Scale Curve=&Lerp_Float(A=1, B=1, Alpha=&Add_Float(A=$System.BuildUp, B=$Particles.OtherAnim))
     )))
   &ScaleColor(
    Scale RGB=&Multiply_VectorByFloat(
     Vector=&MakeVectorFromLinearColorRGB(Color=&LerpLinearColors(
       StartColor=<rgba 0.9244 0.068 1 1>
       EndColor=<rgba 0 0.7442 2 1>
       LerpFactor=&FloatFromCurve(
        FloatCurve=DataInterface() {propertyValues {Curve {keys [
            {interpMode=RCIM_Cubic, tangentMode=RCTM_User, time=0, value=0}
            {interpMode=RCIM_Cubic, tangentMode=RCTM_User, time=0.5, value=1}
            {tangentMode=RCTM_User, time=1, value=0}
           ]}}}
        Scale Curve=0.5
       )
      ))
     Float=&FloatFromCurve(
      FloatCurve=DataInterface() {propertyValues {Curve {keys [
          {interpMode=RCIM_Cubic, tangentMode=RCTM_User, time=0, value=0}
          {interpMode=RCIM_Cubic, tangentMode=RCTM_User, time=0.5, value=1}
          {tangentMode=RCTM_User, time=1, value=0}
         ]}}}
      Scale Curve=0.5
     )
    )
    ScaleA=false
    ScaleRGB=true
   )
   /SetVariables(Particles.VisibilityTag=&MakeCustomIntFromBool(Bool=&SetBoolByFloatComparison(A=$System.Active, B=0.05), True INT=0, False INT=1)) {isSetParameters=true}
  ]
  rendererClasses [&NiagaraMeshRendererProperties, &NiagaraSpriteRendererProperties]
  renderers [
   {rendererClass=&NiagaraMeshRendererProperties, rendererIndex=0}
   {rendererClass=&NiagaraSpriteRendererProperties, rendererIndex=1}
  ]
 }
 VerticalSpritesLow {
  emitterUpdateScript [
   &EmitterState(
    Inactive Response=::NewEnumerator1
    Life Cycle Mode=::NewEnumerator1
    Loop Behavior=::NewEnumerator1
    Loop Duration Mode=::NewEnumerator1
    UseLoopDelay=false
   )
   &SpawnBurst_Instantaneous(Spawn Count=6, Use Spawn Probability=false, UseLoopCountLimit=false)
  ]
  particleSpawnScript [
   &FN_SetSpawnGroup(Inclusivity=true, Normalize Spawn Group=true, Number of Spawn Groups=1, Particles Per Group=6)
   /SetVariables(Particles.SpawnGroups0-1:NiagaraFloat=&RemapRange(
     Input Value To Remap=&MakeFloatFromInt(Int=&SpawnGroup())
     Input Min=0
     Input Max=5
     Output Min=0
     Output Max=1
    )) {isSetParameters=true, module=SetVariables_#12}
   /SetVariables(Particles.SpawnGroupsCurve:NiagaraFloat=&FloatFromCurve(
     FloatCurve=DataInterface() {propertyValues {Curve {keys [
         {interpMode=RCIM_Cubic, tangentMode=RCTM_User, time=0, value=0}
         {interpMode=RCIM_Cubic, tangentMode=RCTM_User, time=0.5, value=1}
         {tangentMode=RCTM_User, time=1, value=0}
        ]}}}
     CurveIndex=$Particles.Normalized Spawn Group
    )) {isSetParameters=true}
   &InitializeParticle(
    Color=<rgba 1 1 1 0.1>
    Lifetime=5
    Position Mode=::NewEnumerator1
    Position Offset=<0 0 50>
    Sprite Size=<1 1>
    Sprite Size Mode=::NewEnumerator4
    UsePositionOffset=false
    Write Material Random=true
   )
   &DynamicMaterialParameters(
    Index 0 Param 1=2
    Index 0 Param 2=25
    Index 0 Param 3=5
    Module.Param0WriteEnabled=true
    Module.Param1WriteEnabled=true
    Module.Param2WriteEnabled=true
    Module.Param3WriteEnabled=false
   )
  ]
  particleUpdateScript [
   &ParticleState(Kill Particles When Lifetime Has Elapsed=false)
   /SetVariables(Particles.BeatAnim:NiagaraFloat=&FloatFromCurve(
     FloatCurve=DataInterface() {propertyValues {Curve {keys [
         {arriveTangent=99.8455, interpMode=RCIM_Cubic, leaveTangent=99.8455, tangentMode=RCTM_User, time=0, value=0}
         {arriveTangent=-5.2513, interpMode=RCIM_Cubic, leaveTangent=-5.2513, tangentMode=RCTM_User, time=0.025, value=1}
         <at 0.5 0>
        ]}}}
     CurveIndex=&Add_Float(
      A=$System.BeatTime
      B=&Lerp_Float(
       A=0
       B=&ClampFloat(Float=&Multiply_Float(A=$System.BeatSpeed, B=-0.125), Min=-0.2, Max=0)
       Alpha=Linked() {linkedVariable {name=Particles.SpawnGroups0-1, type=NiagaraFloat}}
      )
     )
    )) {isSetParameters=true, module=SetVariables_#13}
   /SetVariables(Particles.DropAnim:NiagaraFloat=&FloatFromCurve(
     FloatCurve=DataInterface() {propertyValues {Curve {keys [
         {arriveTangent=99.8455, interpMode=RCIM_Cubic, leaveTangent=99.8455, tangentMode=RCTM_User, time=0, value=0}
         {arriveTangentWeight=0.1307, interpMode=RCIM_Cubic, leaveTangentWeight=0.1859, tangentMode=RCTM_User, time=0.025, value=1}
         {arriveTangent=-4.7766, interpMode=RCIM_Cubic, leaveTangent=-4.7766, tangentMode=RCTM_User, time=0.0649, value=0.9134}
         <at 0.7 0>
        ]}}}
     CurveIndex=&Add_Float(
      A=$System.DropTime
      B=&Lerp_Float(A=0, B=-0.125, Alpha=Linked() {linkedVariable {name=Particles.SpawnGroups0-1, type=NiagaraFloat}})
     )
     Scale Curve=2
    )) {isSetParameters=true}
   /SetVariables(Particles.PositionOrigin:NiagaraPosition=&AddVectorToPosition(
     Position=&LerpPosition(
      A=$System.BeamOrigin
      B=&AddVectorToPosition(
       Position=$System.HeadPosition
       Vector=&MakeVector(X=&Lerp_Float(A=2000, B=1650, Alpha=$Particles.SpawnGroupsCurve), Y=0, Z=0)
      )
      Alpha=&Lerp_Float(
       A=0.87
       B=&Lerp_Float(A=0.94, B=1, Alpha=&Add_Float(A=$Particles.BeatAnim, B=$Particles.DropAnim))
       Alpha=Linked() {linkedVariable {name=Particles.SpawnGroups0-1, type=NiagaraFloat}}
      )
     )
     Vector=&MakeVector(
      X=0
      Y=&FloatFromCurve(
       FloatCurve=DataInterface() {propertyValues {Curve {keys [{tangentMode=RCTM_User, time=0, value=-1}, <at 0.4 -0.4>, <at 0.6 0.4>, {tangentMode=RCTM_User, time=1, value=1}]}}}
       CurveIndex=$Particles.Normalized Spawn Group
       Scale Curve=&Multiply_Float(
        A=&Multiply_Float(
         A=&Lerp_Float(A=1370, B=1400, Alpha=Linked() {linkedVariable {name=Particles.SpawnGroups0-1, type=NiagaraFloat}})
         B=&Add_Float(A=&Lerp_Float(A=1, B=0.7, Alpha=$System.BuildUp), B=$Particles.DropAnim)
        )
        B=$System.Active
       )
      )
      Z=0
     )
    )) {isSetParameters=true, module=SetVariables_#14}
   /SetVariables(Particles.AngleAnim:NiagaraFloat=&Add_Float(
     A=0
     B=&FloatFromCurve(
      FloatCurve=DataInterface() {propertyValues {Curve {keys [{arriveTangent=3.6454, interpMode=RCIM_Cubic, time=0, value=0}, {interpMode=RCIM_Cubic, tangentMode=RCTM_User, time=0.45, value=1}]}}}
      CurveIndex=&Add_Float(A=$System.OtherTime, B=&Lerp_Float(A=-0.2, B=0, Alpha=$Particles.SpawnGroupsCurve))
     )
    )) {isSetParameters=true}
   /SetVariables(Particles.Offset:Vector3f=&RotateVectorByQuaternion(
     Vector To Rotate=<-1500 0 0>
     Quaternion=&MakeQuatFromAxisAngle(Axis=<0 1 0>, AngleInDegrees=&Multiply_Float(A=$Particles.AngleAnim, B=360))
    )) {isSetParameters=true, module=SetVariables_#17}
   /SetVariables(Particles.Position=&AddVectorToPosition(Position=$Particles.PositionOrigin, Vector=$Particles.Offset)) {isSetParameters=true, module=SetVariables_#18}
   /SetVariables(Particles.SpriteSize=&MakeVector2D(
     X=34
     Y=&Multiply_Float(
      A=&Multiply_Float(
       A=&Lerp_Float(A=250, B=400, Alpha=$Particles.SpawnGroupsCurve)
       B=&Lerp_Float(A=1, B=1.5, Alpha=$System.BuildUp)
      )
      B=&Lerp_Float(A=1, B=1.5, Alpha=$Particles.BeatAnim)
     )
    )) {isSetParameters=true, module=SetVariables_#19}
   &SpriteFacingAndAlignment(
    Sprite Alignment=&RotateVectorByQuaternion(
     Vector To Rotate=<0 0 1>
     Quaternion=&MakeQuatFromAxisAngle(Axis=<0 1 0>, AngleInDegrees=&Multiply_Float(A=$Particles.AngleAnim, B=360))
    )
    WriteAlignment=true
    WriteFacing=false
   )
   &ScaleColor(
    Scale Alpha=&FloatFromCurve(
     FloatCurve=DataInterface() {propertyValues {Curve {keys [<at 0 1 RCIM_Constant>,<at 0.1 0.2>,<at 1 0.02>]}}}
     CurveIndex=Linked() {linkedVariable {name=Particles.SpawnGroups0-1, type=NiagaraFloat}}
     Scale Curve=&FloatFromCurve(
      FloatCurve=DataInterface() {propertyValues {Curve {keys [
          {arriveTangent=-9.5354, interpMode=RCIM_Cubic, leaveTangent=-9.5354, tangentMode=RCTM_User, time=0, value=1}
          {interpMode=RCIM_Cubic, tangentMode=RCTM_User, time=0.3, value=0}
          {arriveTangentWeight=0.2325, interpMode=RCIM_Cubic, leaveTangentWeight=0.1744, tangentMode=RCTM_User, time=0.7, value=0}
          {arriveTangent=6.5634, interpMode=RCIM_Cubic, leaveTangent=6.5634, tangentMode=RCTM_User, time=1, value=1}
         ]}}}
      CurveIndex=$Particles.AngleAnim
      Scale Curve=$System.Active
     )
    )
    Scale RGB=&Multiply_VectorByFloat(
     Vector=&MakeVectorFromLinearColorRGB(Color=&LerpLinearColors(StartColor=<rgba 0.9244 0.068 1 1>, EndColor=<rgba 0 0.5582 1.5 1>, LerpFactor=$System.BuildUp))
     Float=&Lerp_Float(A=0.7, B=10, Alpha=&Add_Float(A=$Particles.BeatAnim, B=$Particles.DropAnim))
    )
    ScaleA=true
    ScaleRGB=true
   )
   /SetVariables(Particles.VisibilityTag=&MakeCustomIntFromBool(Bool=&SetBoolByFloatComparison(A=$System.Active, B=0.05), True INT=0, False INT=1)) {isSetParameters=true}
  ]
  rendererClasses [&NiagaraSpriteRendererProperties]
  renderers [{rendererClass=&NiagaraSpriteRendererProperties}]
 }
}
summary {
 systemName=NS_FrontEnd_S2220_DJ
 emitters [
  {emitterName=Flare, rendererClasses [&NiagaraSpriteRendererProperties]}
  {emitterName=Spotlight, rendererClasses [&NiagaraMeshRendererProperties, &NiagaraSpriteRendererProperties]}
  {emitterName=BeamsLong, rendererClasses [&NiagaraMeshRendererProperties]}
  {emitterName=HexSprites, rendererClasses [&NiagaraSpriteRendererProperties]}
  {bEnabled=false, emitterName=SpotlightBits, simTarget=GPUComputeSim, rendererClasses [&NiagaraSpriteRendererProperties]}
  {emitterName=StageBeams, rendererClasses [&NiagaraMeshRendererProperties, &NiagaraSpriteRendererProperties]}
  {emitterName=HexSpritesLow, rendererClasses [&NiagaraSpriteRendererProperties]}
  {emitterName=VerticalSpritesLow, rendererClasses [&NiagaraSpriteRendererProperties]}
  {emitterName=BeamsLong001, rendererClasses [&NiagaraMeshRendererProperties]}
 ]
 userVariables [User.SkelMesh1 {
   type=NiagaraDataInterfaceSkeletalMesh
   defaultValue {
    struct=NiagaraExt_VariableValue_DataInterface
    value {dataInterface=&self:NiagaraDataInterfaceSkeletalMesh_0}
   }
  }]
}
