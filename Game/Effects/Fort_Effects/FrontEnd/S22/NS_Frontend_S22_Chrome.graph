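# Niagara system for the Season 22 frontend chrome effect, layering rising blobs, colored glows and swirling ribbons over a wide area.
# GPU blob sprites drift on curl noise; zero-scaled Blob_Spawn particles seed a secondary blob emitter through particle reads.
# Zero-size RibbonParent particles move under vortex forces and spawn pale RibbonChild trails.
# Blob spawn rates ramp via User.RampupTimeDebris, glow rates via loop age, and GPU emitters scale spawning by distance.
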NiagaraSystem /Game/Effects/Fort_Effects/FrontEnd/S22/NS_Frontend_S22_Chrome
dependencies {
 usedDataInterfaces [&NiagaraDataInterfaceCurve, &NiagaraDataInterfaceParticleRead]
 usedDynamicInputs []
 usedModules [&SystemState]
}
emitters {
 Blob_Spawn {
  simTarget=GPUComputeSim
  emitterUpdateScript [
   &EmitterState(
    Enable Distance Culling=true
    MaxDistance=50000
    Module.bApplySpawnCountScaleByDistance=true
    Module.bUseMaxDistance=true
    Module.bUseMinDistance=false
    Scalability Mode=::NewEnumerator1
    Scale Spawn Count=true
    Spawn Count Scale By Distance=DataInterface() {propertyValues {Curve {keys [<at 0 1>,<at 0.25 1>,<at 1 0>]}}}
   )
   &SpawnRate(
    SpawnRate=&Multiply_Float(
     A=50
     B=&FloatFromCurve(
      FloatCurve=DataInterface() {propertyValues {Curve {keys [<at 0 0>,<at 1 1>]}}}
      CurveIndex=&Divide_Float(A=$System.Age, B=$User.RampupTimeDebris)
     )
    )
    Use Spawn Probability=false
   )
  ]
  particleSpawnScript [
   &InitializeParticle(
    Color=<rgba 1 0 0.8587 2>
    Lifetime=&RandomRangeFloat(Minimum=3, Maximum=5)
    Sprite Rotation Mode=::NewEnumerator2
    Sprite Size Mode=::NewEnumerator1
    Uniform Sprite Size Max=5
    UsePositionOffset=false
    Write Material Random=true
   )
   &TorusLocation(Align Torus With Axis=false, MaskSpawnGroup=false, Module.Override Randomness=false, Torus Mode=::NewEnumerator2)
   &ShapeLocation(Box Size=<1200 200 10>, MaskSpawnGroup=false, Override Seed=false, Shape Primitive=::NewEnumerator2, Surface Only=false, UseOwnerScale=true)
   &SubUVAnimation(End Frame=3, SubUV Animation Mode:ENiagara_SubUVLookupMode=::NewEnumerator2) {module=SubUVAnimation001}
   /SetVariables(Particles.SpriteSize=&Vector2DFromFloat(Value=&RandomRangeFloat(
      Minimum=3
      Maximum=&UniformAOrBFloat(A=10, B=&RandomRangeFloat(Minimum=20, Maximum=30), Distribution Weight=0.25)
     ))) {isSetParameters=true, module=SetVariables_#1}
   &CameraOffset(Camera Offset Amount=-50)
  ]
  particleUpdateScript [
   &ParticleState()
   &GravityForce(Gravity=<0 0 25>)
   &DynamicMaterialParameters(
    Index 0 Param 1=&RandomRangeFloat(Minimum=-0.2, Maximum=1)
    Index 0 Param 2=1
    Index 0 Param 3=&RandomRangeFloat(Minimum=0, Maximum=0.5)
    Index 0 Param 4=1
    Index 1 Param 1=&RandomRangeFloat(Minimum=1, Maximum=2)
    Index 1 Param 2=&FloatFromCurve(FloatCurve=DataInterface() {propertyValues {Curve {keys [<at 0 0.9>,<at 1 -1>]}}}, Scale Curve=0.25)
    Index 1 Param 3=&FloatFromCurve(
     FloatCurve=DataInterface() {propertyValues {Curve {keys [
         <at 0 0>
         {interpMode=RCIM_Cubic, tangentMode=RCTM_User, time=0.05, value=-0.05}
         {interpMode=RCIM_Cubic, tangentMode=RCTM_User, time=0.1, value=1.6}
         {interpMode=RCIM_Cubic, tangentMode=RCTM_User, time=0.2, value=0.1}
         {interpMode=RCIM_Cubic, tangentMode=RCTM_User, time=0.5, value=0.75}
         {interpMode=RCIM_Cubic, tangentMode=RCTM_User, time=0.6, value=0.7}
         {interpMode=RCIM_Cubic, tangentMode=RCTM_User, time=1, value=1.75}
        ]}}}
     Scale Curve=&RandomRangeFloat(Minimum=0.1, Maximum=0.25)
    )
    Index 3 Param 3=&RandomRangeFloat(Minimum=800, Maximum=1500)
    Module.Param0WriteEnabled=true
    Module.Param0WriteEnabled1=true
    Module.Param0WriteEnabled3=true
    Module.Param1WriteEnabled=true
    Module.Param1WriteEnabled1=true
    Module.Param1WriteEnabled3=true
    Module.Param2WriteEnabled=true
    Module.Param2WriteEnabled1=true
    Module.Param2WriteEnabled3=true
    Module.Param3WriteEnabled=true
    Module.Param3WriteEnabled1=true
    Module.Param3WriteEnabled3=true
    Write Parameter Index 1=true
    Write Parameter Index 3=true
   )
   &CurlNoiseForce(Noise Frequency=5, Noise Strength=50, Pan Noise=true, Pan Noise Field=<0.2 1 0.2>)
   &SolveForcesAndVelocity(Clamp Velocity=true, Limit Acceleration=false, Speed Limit=50)
   &ScaleSpriteSizeBySpeed(
    Max Scale Factor=<1 2>
    Min Scale Factor=<0.5 0.5>
    Sample Scale Factor By Curve=true
    Scale Factor Curve=DataInterface() {propertyValues {Curve {keys [{interpMode=RCIM_Cubic, tangentMode=RCTM_User, time=0, value=0}, {interpMode=RCIM_Cubic, tangentMode=RCTM_User, time=1, value=1}]}}}
    Velocity Threshold=500
   )
   &ScaleSpriteSize(
    Uniform Curve Scale=&Multiply_Float(
     A=0
     B=&FloatFromCurve(
      FloatCurve=DataInterface() {propertyValues {Curve {keys [
          {interpMode=RCIM_Cubic, tangentMode=RCTM_User, time=0, value=1}
          <at 0.7 0.9>
          {interpMode=RCIM_Cubic, tangentMode=RCTM_User, time=1, value=0}
         ]}}}
      Scale Curve=&Waveform(
       Waveform Count=::NewEnumerator1
       Global Amplitude Scale=5
       "Waveform [1]"=::NewEnumerator0
       "[1] Phase Scale"=&RandomRangeFloat(Minimum=0.7, Maximum=1)
       "[1] Frequency"=1
       "[1] Amplitude Min/Max"=<0.1 1>
       "[2] Average"=false
       "[2] Input Position"=0
       "[2] Phase"=$Emitter.Age
       "[2] Phase Scale"=0.25
       "[2] Compound Frequency"=2
       "[2] Frequency"=&RandomRangeFloat(Minimum=1.25, Maximum=2.5)
       "[2] Amplitude Min/Max"=<0 1>
       "[2] Scale"=1
       "[2] Bias"=0
       "[2] Exponent"=1
       "[2] Pulse Duration"=0.1
       "[2] Clip Triangle Min/Max"=<0 0>
       "[2] Random Seed"=0
      )
     )
    )
    Uniform Curve Sprite Scale=DataInterface() {propertyValues {Curve {keys [
        {interpMode=RCIM_Cubic, tangentMode=RCTM_User, time=0, value=0}
        {interpMode=RCIM_Cubic, tangentMode=RCTM_User, time=0.1, value=1}
        {interpMode=RCIM_Cubic, tangentMode=RCTM_User, time=0.8, value=1}
        {tangentMode=RCTM_User, time=1, value=0.05}
       ]}}}
   ) {module=ScaleSpriteSize001}
  ]
  rendererClasses [&NiagaraSpriteRendererProperties]
  renderers [{rendererClass=&NiagaraSpriteRendererProperties}]
 }
 Blobs_Bottom {
  simTarget=GPUComputeSim
  emitterUpdateScript [
   &EmitterState(
    Enable Distance Culling=true
    MaxDistance=50000
    Module.bApplySpawnCountScaleByDistance=true
    Module.bUseMaxDistance=true
    Module.bUseMinDistance=false
    Scalability Mode=::NewEnumerator1
    Scale Spawn Count=true
    Spawn Count Scale By Distance=DataInterface() {propertyValues {Curve {keys [<at 0 1>,<at 0.25 1>,<at 1 0>]}}}
   )
   &SpawnRate(
    SpawnRate=&Multiply_Float(
     A=1500
     B=&FloatFromCurve(
      FloatCurve=DataInterface() {propertyValues {Curve {keys [<at 0 0>,<at 1 1>]}}}
      CurveIndex=&Divide_Float(A=$System.Age, B=$User.RampupTimeDebris)
     )
    )
    Use Spawn Probability=false
   )
  ]
  particleSpawnScript [
   &InitializeParticle(
    Color=<rgba 1 0 0.8587 2>
    Lifetime=&RandomRangeFloat(Minimum=5, Maximum=8)
    Sprite Rotation Mode=::NewEnumerator2
    Sprite Size Mode=::NewEnumerator1
    Uniform Sprite Size Max=0.5
    Uniform Sprite Size Min=0.25
    UsePositionOffset=false
    Write Material Random=true
   )
   &TorusLocation(Align Torus With Axis=false, MaskSpawnGroup=false, Module.Override Randomness=false, Torus Mode=::NewEnumerator2)
   &ShapeLocation(Box Size=<1200 500 10>, MaskSpawnGroup=false, Override Seed=false, Shape Primitive=::NewEnumerator2, Surface Only=false, UseOwnerScale=true)
   &SubUVAnimation(End Frame=3, SubUV Animation Mode:ENiagara_SubUVLookupMode=::NewEnumerator2) {module=SubUVAnimation001}
   /SetVariables(Particles.SpriteSize=&Vector2DFromFloat(Value=&RandomRangeFloat(
      Minimum=3
      Maximum=&UniformAOrBFloat(A=10, B=&RandomRangeFloat(Minimum=20, Maximum=30), Distribution Weight=0.25)
     ))) {isSetParameters=true, module=SetVariables_#1}
   &CameraOffset(Camera Offset Amount=-50)
  ]
  particleUpdateScript [
   &ParticleState()
   &GravityForce(Gravity=<0 0 10>)
   &DynamicMaterialParameters(
    Index 0 Param 1=&RandomRangeFloat(Minimum=-0.2, Maximum=1)
    Index 0 Param 2=1
    Index 0 Param 3=&RandomRangeFloat(Minimum=0, Maximum=0.5)
    Index 0 Param 4=1
    Index 1 Param 1=&RandomRangeFloat(Minimum=1, Maximum=2)
    Index 1 Param 2=&FloatFromCurve(FloatCurve=DataInterface() {propertyValues {Curve {keys [<at 0 0.9>,<at 1 -1>]}}}, Scale Curve=0.25)
    Index 1 Param 3=&FloatFromCurve(
     FloatCurve=DataInterface() {propertyValues {Curve {keys [
         <at 0 0>
         {interpMode=RCIM_Cubic, tangentMode=RCTM_User, time=0.05, value=-0.05}
         {interpMode=RCIM_Cubic, tangentMode=RCTM_User, time=0.1, value=1.6}
         {interpMode=RCIM_Cubic, tangentMode=RCTM_User, time=0.2, value=0.1}
         {interpMode=RCIM_Cubic, tangentMode=RCTM_User, time=0.5, value=0.75}
         {interpMode=RCIM_Cubic, tangentMode=RCTM_User, time=0.6, value=0.7}
         {interpMode=RCIM_Cubic, tangentMode=RCTM_User, time=1, value=1.75}
        ]}}}
     Scale Curve=&RandomRangeFloat(Minimum=0.1, Maximum=0.25)
    )
    Index 3 Param 3=&RandomRangeFloat(Minimum=800, Maximum=1500)
    Module.Param0WriteEnabled=true
    Module.Param0WriteEnabled1=true
    Module.Param0WriteEnabled3=true
    Module.Param1WriteEnabled=true
    Module.Param1WriteEnabled1=true
    Module.Param1WriteEnabled3=true
    Module.Param2WriteEnabled=true
    Module.Param2WriteEnabled1=true
    Module.Param2WriteEnabled3=true
    Module.Param3WriteEnabled=true
    Module.Param3WriteEnabled1=true
    Module.Param3WriteEnabled3=true
    Write Parameter Index 1=true
    Write Parameter Index 3=true
   )
   &CurlNoiseForce(Noise Frequency=5, Noise Strength=50, Pan Noise=true, Pan Noise Field=<0.2 1 0.2>)
   &SolveForcesAndVelocity(Clamp Velocity=true, Limit Acceleration=false, Speed Limit=35)
   &ScaleSpriteSizeBySpeed(
    Max Scale Factor=<1 2>
    Min Scale Factor=<0.5 0.5>
    Sample Scale Factor By Curve=true
    Scale Factor Curve=DataInterface() {propertyValues {Curve {keys [{interpMode=RCIM_Cubic, tangentMode=RCTM_User, time=0, value=0}, {interpMode=RCIM_Cubic, tangentMode=RCTM_User, time=1, value=1}]}}}
    Velocity Threshold=500
   )
   &ScaleSpriteSize(
    Uniform Curve Scale=&Multiply_Float(
     A=0.1
     B=&FloatFromCurve(
      FloatCurve=DataInterface() {propertyValues {Curve {keys [
          {interpMode=RCIM_Cubic, tangentMode=RCTM_User, time=0, value=1}
          <at 0.7 0.9>
          {interpMode=RCIM_Cubic, tangentMode=RCTM_User, time=1, value=0}
         ]}}}
      Scale Curve=&Waveform(
       Waveform Count=::NewEnumerator1
       Global Amplitude Scale=5
       "Waveform [1]"=::NewEnumerator0
       "[1] Phase Scale"=&RandomRangeFloat(Minimum=0.7, Maximum=1)
       "[1] Frequency"=1
       "[1] Amplitude Min/Max"=<0.1 1>
       "[2] Average"=false
       "[2] Input Position"=0
       "[2] Phase"=$Emitter.Age
       "[2] Phase Scale"=0.25
       "[2] Compound Frequency"=2
       "[2] Frequency"=&RandomRangeFloat(Minimum=1.25, Maximum=2.5)
       "[2] Amplitude Min/Max"=<0 1>
       "[2] Scale"=1
       "[2] Bias"=0
       "[2] Exponent"=1
       "[2] Pulse Duration"=0.1
       "[2] Clip Triangle Min/Max"=<0 0>
       "[2] Random Seed"=0
      )
     )
    )
    Uniform Curve Sprite Scale=DataInterface() {propertyValues {Curve {keys [
        {interpMode=RCIM_Cubic, tangentMode=RCTM_User, time=0, value=0}
        {interpMode=RCIM_Cubic, tangentMode=RCTM_User, time=0.1, value=1}
        {interpMode=RCIM_Cubic, tangentMode=RCTM_User, time=0.8, value=1}
        {tangentMode=RCTM_User, time=1, value=0.05}
       ]}}}
   ) {module=ScaleSpriteSize001}
  ]
  rendererClasses [&NiagaraSpriteRendererProperties]
  renderers [{rendererClass=&NiagaraSpriteRendererProperties}]
 }
 Blobs_Bottom003 {
  simTarget=GPUComputeSim
  emitterUpdateScript [
   &EmitterState(
    Enable Distance Culling=true
    MaxDistance=50000
    Module.bApplySpawnCountScaleByDistance=true
    Module.bUseMaxDistance=true
    Module.bUseMinDistance=false
    Scalability Mode=::NewEnumerator1
    Scale Spawn Count=true
    Spawn Count Scale By Distance=DataInterface() {propertyValues {Curve {keys [<at 0 1>,<at 0.25 1>,<at 1 0>]}}}
   )
   &SpawnParticlesFromOtherEmitter(
    Attribute Reader=DataInterface() {propertyValues {EmitterBinding {bindingMode=Other, emitterName=Blob_Spawn}}}
    Spawn Rate=20
    Spawn Rate Per Particle Cap=20
    Use Maximum Age=false
    Use Minimum Age=false
   )
  ]
  particleSpawnScript [
   &InitializeParticle(
    Color=<rgba 1 0 0.8587 2>
    Lifetime=&RandomRangeFloat(Minimum=3, Maximum=5)
    Sprite Rotation Mode=::NewEnumerator2
    Sprite Size Mode=::NewEnumerator1
    Uniform Sprite Size Max=5
    UsePositionOffset=false
    Write Material Random=true
   )
   &SubUVAnimation(End Frame=3, SubUV Animation Mode:ENiagara_SubUVLookupMode=::NewEnumerator2) {module=SubUVAnimation001}
   /SetVariables(Particles.SpriteSize=&Vector2DFromFloat(Value=&RandomRangeFloat(
      Minimum=3
      Maximum=&UniformAOrBFloat(A=10, B=&RandomRangeFloat(Minimum=20, Maximum=30), Distribution Weight=0.25)
     ))) {isSetParameters=true, module=SetVariables_#1}
   &CameraOffset(Camera Offset Amount=-50)
   &SampleParticlesFromOtherEmitter(
    Maximum Sampled Particle Age=2
    Minimum Sampled Particle Age=1
    Module.MaxSampledAgeCutoff=true
    Module.MinSampledAgeCutoff=true
    Reject Particle By Sampled Age=true
    Velocity Sampling=::NewEnumerator0
   )
  ]
  particleUpdateScript [
   &ParticleState()
   &GravityForce(Gravity=<0 0 25>)
   &DynamicMaterialParameters(
    Index 0 Param 1=&RandomRangeFloat(Minimum=-0.2, Maximum=1)
    Index 0 Param 2=1
    Index 0 Param 3=&RandomRangeFloat(Minimum=0, Maximum=0.5)
    Index 0 Param 4=1
    Index 1 Param 1=&RandomRangeFloat(Minimum=1, Maximum=2)
    Index 1 Param 2=&FloatFromCurve(FloatCurve=DataInterface() {propertyValues {Curve {keys [<at 0 0.9>,<at 1 -1>]}}}, Scale Curve=0.25)
    Index 1 Param 3=&FloatFromCurve(
     FloatCurve=DataInterface() {propertyValues {Curve {keys [
         <at 0 0>
         {interpMode=RCIM_Cubic, tangentMode=RCTM_User, time=0.05, value=-0.05}
         {interpMode=RCIM_Cubic, tangentMode=RCTM_User, time=0.1, value=1.6}
         {interpMode=RCIM_Cubic, tangentMode=RCTM_User, time=0.2, value=0.1}
         {interpMode=RCIM_Cubic, tangentMode=RCTM_User, time=0.5, value=0.75}
         {interpMode=RCIM_Cubic, tangentMode=RCTM_User, time=0.6, value=0.7}
         {interpMode=RCIM_Cubic, tangentMode=RCTM_User, time=1, value=1.75}
        ]}}}
     Scale Curve=&RandomRangeFloat(Minimum=0.1, Maximum=0.25)
    )
    Index 3 Param 3=&RandomRangeFloat(Minimum=800, Maximum=1500)
    Module.Param0WriteEnabled=true
    Module.Param0WriteEnabled1=true
    Module.Param0WriteEnabled3=true
    Module.Param1WriteEnabled=true
    Module.Param1WriteEnabled1=true
    Module.Param1WriteEnabled3=true
    Module.Param2WriteEnabled=true
    Module.Param2WriteEnabled1=true
    Module.Param2WriteEnabled3=true
    Module.Param3WriteEnabled=true
    Module.Param3WriteEnabled1=true
    Module.Param3WriteEnabled3=true
    Write Parameter Index 1=true
    Write Parameter Index 3=true
   )
   &CurlNoiseForce(Noise Frequency=25, Noise Strength=50, Pan Noise=true, Pan Noise Field=<0.2 1 0.2>)
   &PointForce(
    Force Origin=$Output.SampleParticlesFromOtherEmitter.SampledPosition
    Force Strength=50
    Linear Force Vector:Vector3f=<0 0 1>
    Module.Use Falloff=false
    Module.UseLinearForce=true
   )
   &Drag(Drag=0.2, Use Linear Drag=true, Use Rotational Drag=true)
   &SolveForcesAndVelocity(Clamp Velocity=true, Limit Acceleration=false, Speed Limit=25)
   &ScaleSpriteSizeBySpeed(
    Max Scale Factor=<1 2>
    Min Scale Factor=<0.5 0.5>
    Sample Scale Factor By Curve=true
    Scale Factor Curve=DataInterface() {propertyValues {Curve {keys [{interpMode=RCIM_Cubic, tangentMode=RCTM_User, time=0, value=0}, {interpMode=RCIM_Cubic, tangentMode=RCTM_User, time=1, value=1}]}}}
    Velocity Threshold=500
   )
   &ScaleSpriteSize(
    Uniform Curve Scale=&Multiply_Float(
     A=0.1
     B=&FloatFromCurve(
      FloatCurve=DataInterface() {propertyValues {Curve {keys [
          {interpMode=RCIM_Cubic, tangentMode=RCTM_User, time=0, value=1}
          <at 0.7 0.9>
          {interpMode=RCIM_Cubic, tangentMode=RCTM_User, time=1, value=0}
         ]}}}
      Scale Curve=&Waveform(
       Waveform Count=::NewEnumerator1
       Global Amplitude Scale=5
       "Waveform [1]"=::NewEnumerator0
       "[1] Phase Scale"=&RandomRangeFloat(Minimum=0.7, Maximum=1)
       "[1] Frequency"=1
       "[1] Amplitude Min/Max"=<0.1 1>
       "[2] Average"=false
       "[2] Input Position"=0
       "[2] Phase"=$Emitter.Age
       "[2] Phase Scale"=0.25
       "[2] Compound Frequency"=2
       "[2] Frequency"=&RandomRangeFloat(Minimum=1.25, Maximum=2.5)
       "[2] Amplitude Min/Max"=<0 1>
       "[2] Scale"=1
       "[2] Bias"=0
       "[2] Exponent"=1
       "[2] Pulse Duration"=0.1
       "[2] Clip Triangle Min/Max"=<0 0>
       "[2] Random Seed"=0
      )
     )
    )
    Uniform Curve Sprite Scale=DataInterface() {propertyValues {Curve {keys [
        {interpMode=RCIM_Cubic, tangentMode=RCTM_User, time=0, value=0}
        {interpMode=RCIM_Cubic, tangentMode=RCTM_User, time=0.1, value=1}
        {interpMode=RCIM_Cubic, tangentMode=RCTM_User, time=0.8, value=1}
        {tangentMode=RCTM_User, time=1, value=0.05}
       ]}}}
   ) {module=ScaleSpriteSize001}
  ]
  rendererClasses [&NiagaraSpriteRendererProperties]
  renderers [{rendererClass=&NiagaraSpriteRendererProperties}]
 }
 Glows {
  simTarget=GPUComputeSim
  emitterUpdateScript [
   &EmitterState(
    Enable Distance Culling=true
    MaxDistance=50000
    Module.bApplySpawnCountScaleByDistance=true
    Module.bUseMaxDistance=true
    Module.bUseMinDistance=false
    Scalability Mode=::NewEnumerator1
    Scale Spawn Count=true
    Spawn Count Scale By Distance=DataInterface() {propertyValues {Curve {keys [<at 0 1>,<at 0.25 1>,<at 1 0>]}}}
   )
   &SpawnRate(
    SpawnRate=&Multiply_Float(
     A=150
     B=&FloatFromCurve(FloatCurve=DataInterface() {propertyValues {Curve {keys [<at 0 0>,<at 1 1>]}}}, CurveIndex=$Emitter.NormalizedLoopAge)
    )
    Use Spawn Probability=false
   )
  ]
  particleSpawnScript [
   &InitializeParticle(
    Color=&ColorFromCurve(
     ColorCurve=DataInterface() {propertyValues {
       AlphaCurve {keys [<at 0.2043 1>]}
       BlueCurve {keys [<at 0 1>,<at 0.09 1>,<at 0.208 1>,<at 0.3523 0.3363>,<at 0.521 1>,<at 0.666 1>,<at 1 1>]}
       GreenCurve {keys [<at 0 0>,<at 0.09 0>,<at 0.208 0>,<at 0.3523 1>,<at 0.521 0.3585>,<at 0.666 0>,<at 1 0.3585>]}
       RedCurve {keys [<at 0 0.8831>,<at 0.09 0.9654>,<at 0.208 0.1158>,<at 0.3523 0.1808>,<at 0.521 0>,<at 0.666 0>,<at 1 0>]}
      }}
     CurveIndex=&RandomRangeFloat(Minimum=0, Maximum=1)
    )
    Lifetime=&RandomRangeFloat(Minimum=2, Maximum=5)
    Sprite Rotation Mode=::NewEnumerator2
    Sprite Size Mode=::NewEnumerator1
    Uniform Sprite Size Max=1
    Uniform Sprite Size Min=0.5
    UsePositionOffset=false
    Write Material Random=true
   )
   &TorusLocation(Align Torus With Axis=false, MaskSpawnGroup=false, Module.Override Randomness=false, Torus Mode=::NewEnumerator2)
   &ShapeLocation(Box Size=<1200 500 10>, MaskSpawnGroup=false, Override Seed=false, Shape Primitive=::NewEnumerator2, Surface Only=false, UseOwnerScale=true)
   /SetVariables(Particles.SpriteSize=&Vector2DFromFloat(Value=&RandomRangeFloat(
      Minimum=4
      Maximum=&UniformAOrBFloat(A=6, B=&RandomRangeFloat(Minimum=10, Maximum=20), Distribution Weight=0.1)
     ))) {isSetParameters=true, module=SetVariables_#1}
   &CameraOffset(Camera Offset Amount=-50)
  ]
  particleUpdateScript [
   &ParticleState()
   &GravityForce(Gravity=<0 0 25>)
   &DynamicMaterialParameters(
    Index 0 Param 1=&RandomRangeFloat(Minimum=0.5, Maximum=1)
    Index 0 Param 2=0.1
    Index 0 Param 3=&RandomRangeFloat(Minimum=0, Maximum=0.5)
    Index 0 Param 4=1
    Index 3 Param 3=&RandomRangeFloat(Minimum=800, Maximum=1500)
    Module.Param0WriteEnabled=true
    Module.Param0WriteEnabled3=true
    Module.Param1WriteEnabled=true
    Module.Param1WriteEnabled3=true
    Module.Param2WriteEnabled=true
    Module.Param2WriteEnabled3=true
    Module.Param3WriteEnabled=true
    Module.Param3WriteEnabled3=true
    Write Parameter Index 3=true
   )
   &CurlNoiseForce(Noise Frequency=5, Noise Strength=35, Pan Noise=true, Pan Noise Field=<0.2 1 0.2>)
   &SolveForcesAndVelocity(Clamp Velocity=true, Limit Acceleration=false, Speed Limit=250)
   &ScaleSpriteSize(
    Uniform Curve Scale=&FloatFromCurve(
     FloatCurve=DataInterface() {propertyValues {Curve {keys [<at 2000 0.15>,<at 25000 0.325>]}}}
     CurveIndex=&VectorLength(Vector=&Subtract_Vector(
       A=Linked() {linkedVariable {name=Particles.Position, type=NiagaraPosition}}
       B=&MakeVectorFromLinearColorRGB(Color=$NPC.FortniteNPC.PlayerPosition)
      ))
     Scale Curve=6
    )
    Uniform Curve Sprite Scale=DataInterface() {propertyValues {Curve {keys [
        {interpMode=RCIM_Cubic, tangentMode=RCTM_User, time=0, value=0}
        {interpMode=RCIM_Cubic, tangentMode=RCTM_User, time=0.1, value=1}
        {interpMode=RCIM_Cubic, tangentMode=RCTM_User, time=0.8, value=1}
        {tangentMode=RCTM_User, time=1, value=0.05}
       ]}}}
   ) {enabled=false}
   &ScaleSpriteSize(
    Uniform Curve Scale=&Multiply_Float(
     A=0.5
     B=&FloatFromCurve(
      FloatCurve=DataInterface() {propertyValues {Curve {keys [
          {interpMode=RCIM_Cubic, tangentMode=RCTM_User, time=0, value=1}
          <at 0.7 0.9>
          {interpMode=RCIM_Cubic, tangentMode=RCTM_User, time=1, value=0}
         ]}}}
      Scale Curve=&Waveform(
       Waveform Count=::NewEnumerator1
       Global Amplitude Scale=5
       "Waveform [1]"=::NewEnumerator0
       "[1] Phase Scale"=&RandomRangeFloat(Minimum=1, Maximum=2)
       "[1] Frequency"=1
       "[1] Amplitude Min/Max"=<0.1 1>
       "[2] Average"=false
       "[2] Input Position"=0
       "[2] Phase"=$Emitter.Age
       "[2] Phase Scale"=0.25
       "[2] Compound Frequency"=2
       "[2] Frequency"=&RandomRangeFloat(Minimum=1.25, Maximum=2.5)
       "[2] Amplitude Min/Max"=<0 1>
       "[2] Scale"=1
       "[2] Bias"=0
       "[2] Exponent"=1
       "[2] Pulse Duration"=0.1
       "[2] Clip Triangle Min/Max"=<0 0>
       "[2] Random Seed"=0
      )
     )
    )
    Uniform Curve Sprite Scale=DataInterface() {propertyValues {Curve {keys [
        {interpMode=RCIM_Cubic, tangentMode=RCTM_User, time=0, value=0}
        {interpMode=RCIM_Cubic, tangentMode=RCTM_User, time=0.1, value=1}
        {interpMode=RCIM_Cubic, tangentMode=RCTM_User, time=0.8, value=1}
        {tangentMode=RCTM_User, time=1, value=0.05}
       ]}}}
   ) {module=ScaleSpriteSize001}
   &ScaleColor(
    Scale Alpha=0.25
    Scale RGB=&MakeVectorFromLinearColorRGB(Color=&HueShiftLinearColor(
      Hue=&RandomRangeFloat(
       Minimum=&FloatFromCurve(FloatCurve=DataInterface() {propertyValues {Curve {keys [{tangentMode=RCTM_User, time=0, value=0.5}, {tangentMode=RCTM_User, time=1, value=1}]}}})
       Maximum=&FloatFromCurve(FloatCurve=DataInterface() {propertyValues {Curve {keys [<at 0 1>,<at 1 0.25>]}}})
      )
      Saturation=0.25
      Value=&FloatFromCurve(
       FloatCurve=DataInterface() {propertyValues {Curve {keys [
           {interpMode=RCIM_Cubic, tangentMode=RCTM_User, time=0, value=1}
           {interpMode=RCIM_Cubic, tangentMode=RCTM_User, time=0.2, value=0.45}
           {interpMode=RCIM_Cubic, tangentMode=RCTM_User, time=0.5, value=0.2}
           {interpMode=RCIM_Cubic, tangentMode=RCTM_User, time=0.7, value=0.85}
           {interpMode=RCIM_Cubic, tangentMode=RCTM_User, time=1, value=0}
          ]}}}
       Scale Curve=&Waveform(
        Global Amplitude Scale=50
        "Waveform [1]"=::NewEnumerator0
        "[1] Phase Scale"=&RandomRangeFloat(Minimum=3, Maximum=5)
        "[1] Frequency"=1
        "[1] Amplitude Min/Max"=<0.5 1>
       )
      )
      Linear Color=$Particles.Initial.Color
     ))
    ScaleA=true
    ScaleRGB=true
   )
  ]
  rendererClasses [&NiagaraSpriteRendererProperties]
  renderers [{rendererClass=&NiagaraSpriteRendererProperties}]
 }
 Glows001 {
  simTarget=GPUComputeSim
  emitterUpdateScript [
   &EmitterState(
    Enable Distance Culling=true
    MaxDistance=50000
    Module.bApplySpawnCountScaleByDistance=true
    Module.bUseMaxDistance=true
    Module.bUseMinDistance=false
    Scalability Mode=::NewEnumerator1
    Scale Spawn Count=true
    Spawn Count Scale By Distance=DataInterface() {propertyValues {Curve {keys [<at 0 1>,<at 0.25 1>,<at 1 0>]}}}
   )
   &SpawnRate(
    SpawnRate=&Multiply_Float(
     A=25
     B=&FloatFromCurve(FloatCurve=DataInterface() {propertyValues {Curve {keys [<at 0 0>,<at 1 1>]}}}, CurveIndex=$Emitter.NormalizedLoopAge)
    )
    Use Spawn Probability=false
   )
  ]
  particleSpawnScript [
   &InitializeParticle(
    Color=&ColorFromCurve(
     ColorCurve=DataInterface() {propertyValues {
       AlphaCurve {keys [<at 1.0395 1 RCIM_Cubic>]}
       BlueCurve {keys [
         <at -0.15 1 RCIM_Cubic>
         {arriveTangent=-3.3333, interpMode=RCIM_Cubic, leaveTangent=-3.3333, time=0, value=0.5}
         {arriveTangent=-1, interpMode=RCIM_Cubic, leaveTangent=-1, time=0.15, value=0}
         {arriveTangent=1.4286, interpMode=RCIM_Cubic, leaveTangent=1.4286, time=0.5, value=0}
         {arriveTangent=1, interpMode=RCIM_Cubic, leaveTangent=1, time=0.85, value=1}
         {arriveTangent=-3.3333, interpMode=RCIM_Cubic, leaveTangent=-3.3333, time=1, value=0.5}
         <at 1.15 0 RCIM_Cubic>
        ]}
       GreenCurve {keys [
         <at -0.15 0 RCIM_Cubic>
         <at 0 0 RCIM_Cubic>
         {arriveTangent=2, interpMode=RCIM_Cubic, leaveTangent=2, time=0.15, value=0}
         <at 0.5 1 RCIM_Cubic>
         {arriveTangent=-2, interpMode=RCIM_Cubic, leaveTangent=-2, time=0.85, value=0}
         <at 1 0 RCIM_Cubic>
         <at 1.15 0 RCIM_Cubic>
        ]}
       RedCurve {keys [
         <at -0.15 0 RCIM_Cubic>
         {arriveTangent=3.3333, interpMode=RCIM_Cubic, leaveTangent=3.3333, time=0, value=1}
         {arriveTangent=-2, interpMode=RCIM_Cubic, leaveTangent=-2, time=0.15, value=1}
         {arriveTangent=-1.4286, interpMode=RCIM_Cubic, leaveTangent=-1.4286, time=0.5, value=0}
         {arriveTangent=2, interpMode=RCIM_Cubic, leaveTangent=2, time=0.85, value=0}
         {arriveTangent=3.3333, interpMode=RCIM_Cubic, leaveTangent=3.3333, time=1, value=1}
         <at 1.15 1 RCIM_Cubic>
        ]}
      }}
     CurveIndex=&RandomRangeFloat(Minimum=0, Maximum=1)
    )
    Lifetime=&RandomRangeFloat(Minimum=2, Maximum=5)
    Sprite Rotation Mode=::NewEnumerator2
    Sprite Size Mode=::NewEnumerator1
    Uniform Sprite Size Max=1
    Uniform Sprite Size Min=1
    UsePositionOffset=false
    Write Material Random=true
   )
   &TorusLocation(Align Torus With Axis=false, MaskSpawnGroup=false, Module.Override Randomness=false, Torus Mode=::NewEnumerator2)
   &AddVelocity(Velocity=<0 0 10>)
   &ShapeLocation(Box Size=<1200 200 10>, MaskSpawnGroup=false, Override Seed=false, Shape Primitive=::NewEnumerator2, Surface Only=false, UseOwnerScale=true)
   &CameraOffset()
  ]
  particleUpdateScript [
   &ParticleState()
   &GravityForce(Gravity=<0 0 25>)
   &SolveForcesAndVelocity(Clamp Velocity=true, Limit Acceleration=false, Speed Limit=150)
   &ScaleSpriteSize(
    Uniform Curve Scale=&FloatFromCurve(
     FloatCurve=DataInterface() {propertyValues {Curve {keys [<at 2000 0.15>,<at 25000 0.325>]}}}
     CurveIndex=&VectorLength(Vector=&Subtract_Vector(
       A=Linked() {linkedVariable {name=Particles.Position, type=NiagaraPosition}}
       B=&MakeVectorFromLinearColorRGB(Color=$NPC.FortniteNPC.PlayerPosition)
      ))
     Scale Curve=6
    )
    Uniform Curve Sprite Scale=DataInterface() {propertyValues {Curve {keys [
        {interpMode=RCIM_Cubic, tangentMode=RCTM_User, time=0, value=0}
        {interpMode=RCIM_Cubic, tangentMode=RCTM_User, time=0.1, value=1}
        {interpMode=RCIM_Cubic, tangentMode=RCTM_User, time=0.8, value=1}
        {tangentMode=RCTM_User, time=1, value=0.05}
       ]}}}
   ) {enabled=false}
   &ScaleSpriteSize(
    Uniform Curve Scale=150
    Uniform Curve Sprite Scale=DataInterface() {propertyValues {Curve {keys [
        {interpMode=RCIM_Cubic, tangentMode=RCTM_User, time=0, value=1}
        {interpMode=RCIM_Cubic, tangentMode=RCTM_User, time=0.8, value=1}
        {tangentMode=RCTM_User, time=1, value=0.05}
       ]}}}
   ) {module=ScaleSpriteSize001}
   &ScaleColor(
    Scale Alpha=&FloatFromCurve(
     FloatCurve=DataInterface() {propertyValues {Curve {keys [
         {arriveTangent=9.0725, interpMode=RCIM_Cubic, leaveTangent=9.0725, tangentMode=RCTM_User, time=0, value=0}
         {arriveTangent=-2.5249, interpMode=RCIM_Cubic, leaveTangent=-2.5249, tangentMode=RCTM_User, time=0.1408, value=0.9847}
         {arriveTangent=-2.5973, interpMode=RCIM_Cubic, leaveTangent=-2.5973, tangentMode=RCTM_User, time=0.7, value=0.725}
         {arriveTangent=-0.0405, interpMode=RCIM_Cubic, leaveTangent=-0.0405, tangentMode=RCTM_User, time=1, value=0}
        ]}}}
     Scale Curve=5
    )
    Scale RGB=<3 3 3>
    ScaleA=true
    ScaleRGB=true
   )
  ]
  rendererClasses [&NiagaraSpriteRendererProperties]
  renderers [{rendererClass=&NiagaraSpriteRendererProperties}]
 }
 RibbonChild {
  emitterUpdateScript [
   &EmitterState()
   &SpawnParticlesFromOtherEmitter(
    Attribute Reader=DataInterface() {propertyValues {EmitterBinding {bindingMode=Other, emitterName=RibbonParent}}}
    Spawn Rate Per Particle Cap=25
    Use Maximum Age=false
    Use Minimum Age=false
   )
  ]
  particleSpawnScript [
   &InitializeRibbon(
    Color=<rgba 0.996 0.996 1 1>
    Lifetime=2
    Ribbon Width=&FloatFromCurve(FloatCurve=DataInterface() {propertyValues {Curve {keys [<at 1 1>,<at 2 0>]}}}, Scale Curve=150)
    UsePositionOffset=false
    Write Material Random=true
   )
   &InitializeParticle(Lifetime=2, UsePositionOffset=false, Write Material Random=true) {enabled=false}
   &SampleParticlesFromOtherEmitter(Applied Velocity Scale=0, Particle ID Sampling=::NewEnumerator1, Velocity Sampling=::NewEnumerator0)
   &AddVelocity(Velocity=&RandomRangeVector(Minimum=<0 -5 -5>, Maximum=<1 5 5>))
  ]
  particleUpdateScript [
   &ParticleState()
   &ScaleColor(
    Scale Alpha=&FloatFromCurve(FloatCurve=DataInterface() {propertyValues {Curve {keys [
         {arriveTangent=9.0725, interpMode=RCIM_Cubic, leaveTangent=9.0725, tangentMode=RCTM_User, time=0.3, value=0.05}
         {arriveTangent=0.0667, interpMode=RCIM_Cubic, leaveTangent=0.0667, tangentMode=RCTM_User, time=0.5, value=0.9847}
         {interpMode=RCIM_Cubic, tangentMode=RCTM_User, time=0.6, value=1}
         {arriveTangent=-0.0405, interpMode=RCIM_Cubic, leaveTangent=-0.0405, tangentMode=RCTM_User, time=0.8, value=0}
        ]}}})
    Scale RGB=&Multiply_VectorByFloat(
     Vector=&MakeVectorFromLinearColorRGB(Color=$Particles.Color)
     Float=&FloatFromCurve(FloatCurve=DataInterface() {propertyValues {Curve {keys [<at 0 1>,<at 1 1>]}}}, Scale Curve=0.999)
    )
    ScaleA=true
    ScaleRGB=true
   ) {module=ScaleColor003}
   &ScaleRibbonWidth(Ribbon Width Scale=&FloatFromCurve(
     FloatCurve=DataInterface() {propertyValues {Curve {keys [{interpMode=RCIM_Cubic, tangentMode=RCTM_User, time=0, value=0}, <at 0.2 1.0125>, <at 1 0>]}}}
     Scale Curve=0.7
    ))
   &SolveForcesAndVelocity(Clamp Velocity=false, Limit Acceleration=false)
  ]
  rendererClasses [&NiagaraRibbonRendererProperties]
  renderers [{rendererClass=&NiagaraRibbonRendererProperties}]
 }
 RibbonParent {
  emitterUpdateScript [
   &EmitterState()
   &SpawnRate(SpawnRate=5, Use Spawn Probability=false)
  ]
  particleSpawnScript [
   &InitializeParticle(
    Lifetime=2
    Sprite Rotation Mode=::NewEnumerator2
    Sprite Size Mode=::NewEnumerator1
    Uniform Sprite Size Max=0
    Uniform Sprite Size Min=0
    UsePositionOffset=false
    Write Material Random=true
   )
   &ShapeLocation(Box Size=<800 200 10>, MaskSpawnGroup=false, Override Seed=false, Shape Primitive=::NewEnumerator2, Surface Only=false, UseOwnerScale=true)
   &AddVelocity(Velocity=&UniformRangedVector(Minimum=<-750 -750 0>, Maximum=<750 750 0>)) {enabled=false}
  ]
  particleUpdateScript [
   &ParticleState()
   &VortexForce(
    Influence Falloff Exponent=0
    Influence Falloff Radius=20000
    Origin Pull Amount=200
    Use Influence Falloff=true
    UseOriginOffset=true
    Vortex Force Amount=&RandomRangeFloat(Minimum=-2500, Maximum=-5000)
    Vortex Origin Offset:Vector3f=<0 0 15000>
   )
   &GravityForce(Gravity=<0 0 100>)
   &CurlNoiseForce(Noise Frequency=5, Noise Strength=50, Pan Noise=false)
   &SolveForcesAndVelocity(Clamp Velocity=true, Limit Acceleration=false, Speed Limit=600)
   &ScaleColor(
    Scale Alpha=&FloatFromCurve(FloatCurve=DataInterface() {propertyValues {Curve {keys [<at 0 1>,<at 0.1 1>,<at 1 0>]}}})
    Scale RGB=&VectorFromFloat(Value=&FloatFromCurve(
      FloatCurve=DataInterface() {propertyValues {Curve {keys [
          {interpMode=RCIM_Cubic, tangentMode=RCTM_User, time=0, value=0.25}
          {arriveTangent=-14.2949, interpMode=RCIM_Cubic, leaveTangent=-14.2949, tangentMode=RCTM_User, time=0.2, value=5}
          {interpMode=RCIM_Cubic, tangentMode=RCTM_User, time=1, value=0}
         ]}}}
      Scale Curve=5
     ))
    ScaleA=true
    ScaleRGB=true
   ) {enabled=false}
  ]
  rendererClasses [&NiagaraSpriteRendererProperties]
  renderers [{rendererClass=&NiagaraSpriteRendererProperties}]
 }
}
summary {
 systemName=NS_Frontend_S22_Chrome
 emitters [
  {emitterName=Blobs_Bottom, simTarget=GPUComputeSim, rendererClasses [&NiagaraSpriteRendererProperties]}
  {emitterName=RibbonParent, rendererClasses [&NiagaraSpriteRendererProperties]}
  {emitterName=RibbonChild, rendererClasses [&NiagaraRibbonRendererProperties]}
  {emitterName=Glows, simTarget=GPUComputeSim, rendererClasses [&NiagaraSpriteRendererProperties]}
  {emitterName=Blob_Spawn, simTarget=GPUComputeSim, rendererClasses [&NiagaraSpriteRendererProperties]}
  {emitterName=Blobs_Bottom003, simTarget=GPUComputeSim, rendererClasses [&NiagaraSpriteRendererProperties]}
  {emitterName=Glows001, simTarget=GPUComputeSim, rendererClasses [&NiagaraSpriteRendererProperties]}
 ]
 userVariables [
  User.BaseColor {type=LinearColor, description="", defaultValue {struct=LinearColor, value=<rgba 1 1 1 1>}}
  User.BaseCurveColor {
   type=NiagaraDataInterfaceColorCurve
   description=""
   defaultValue {
    struct=NiagaraExt_VariableValue_DataInterface
    value {
     dataInterface=&self:NiagaraDataInterfaceColorCurve_0
     dataInterfaceClass=&NiagaraDataInterfaceColorCurve
    }
   }
  }
  User.RampupTimeDebris {type=NiagaraFloat, defaultValue=1, description=""}
  User.RampupTimeSmoke {type=NiagaraFloat, defaultValue=1, description=""}
 ]
}
